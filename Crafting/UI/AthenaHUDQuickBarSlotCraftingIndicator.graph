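# HUD quick bar slot widget that shows an animated hammer indicator when the slot's item can be crafted.
# Parent-class events collapse it for non-craftable items and switch between an empty spacer and the indicator as craftability changes.
# Becoming craftable plays a looped animation of hammer, flame, spark and glow material parameters, stopped when craftability is lost.
# Ingredient changes replay the animation if crafting is possible and it is not already playing.
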WidgetBlueprint /Crafting/UI/AthenaHUDQuickBarSlotCraftingIndicator
animations [CanCraftNowAnim {
  binding_count=7
  length_seconds=1
  object=&self:CanCraftNowAnim
  sequence {bindings [
    Flame1 {
     id=#1
     possessed_class=Image
     tracks [{
       type=MovieSceneWidgetMaterialTrack
       sections [{
         type=MovieSceneParameterSection
         channels [
          Scale_Anim {type=MovieSceneScriptingFloatChannel, keys [[0,1],[4,1],[8,0],[13,1]]}
          Offset_X {type=MovieSceneScriptingFloatChannel, const [0,4]}
          Offset_Y {type=MovieSceneScriptingFloatChannel, keys [[0,0],[4,0],[8,0.05],[13,0]]}
          Final_Alpha {type=MovieSceneScriptingFloatChannel, keys [[0,1],[4,0],[8,0],[10,1],[13,0.5],[18,1],[20,1]]}
         ]
        }]
      }]
    }
    Glow {
     id=#2
     possessed_class=Image
     tracks [{
       type=MovieSceneWidgetMaterialTrack
       sections [{type=MovieSceneParameterSection, channels [Final_Alpha {type=MovieSceneScriptingFloatChannel, keys [[0,0],[8,0],[11,1],[20,0]]}]}]
      }]
    }
    Sparks2 {
     id=#3
     possessed_class=Image
     tracks [{
       type=MovieSceneWidgetMaterialTrack
       sections [{
         type=MovieSceneParameterSection
         channels [
          Final_Alpha {type=MovieSceneScriptingFloatChannel, keys [[0,0],[9,0],[14,1],[20,0]]}
          Noise2Scale_Anim {type=MovieSceneScriptingFloatChannel, keys [[0,0],[9,0],[20.0003,1]]}
         ]
        }]
      }]
    }
    Sparks1 {
     id=#4
     possessed_class=Image
     tracks [{
       type=MovieSceneWidgetMaterialTrack
       sections [{
         type=MovieSceneParameterSection
         channels [
          Final_Alpha {type=MovieSceneScriptingFloatChannel, keys [[0,0],[8,0],[10,1],[20.0003,0]]}
          Noise1Scale_Anim {type=MovieSceneScriptingFloatChannel, keys [[0,0],[8,0],[20,1]]}
          Noise1Subtract_Anim {type=MovieSceneScriptingFloatChannel, keys [[0,0],[8,0],[20,1]]}
         ]
        }]
      }]
    }
    Backing_1 {
     id=#5
     possessed_class=Image
     tracks [{
       type=MovieSceneWidgetMaterialTrack
       sections [{type=MovieSceneParameterSection, channels [Animate_Alpha {type=MovieSceneScriptingFloatChannel, keys [[0,0],[7,0],[8,0.25],[20,0]]}]}]
      }]
    }
    Hammer {
     id=#6
     possessed_class=Image
     tracks [{
       type=MovieSceneWidgetMaterialTrack
       sections [{
         type=MovieSceneParameterSection
         channels [
          Alpha_Intensity_Anim {type=MovieSceneScriptingFloatChannel, const [0,1]}
          RotationAngle_Anim {type=MovieSceneScriptingFloatChannel, keys [[0,0],[6,1],[8,-0.075],[11,0.1],[15,0]]}
          Scale_Anim {type=MovieSceneScriptingFloatChannel, const [1,1]}
         ]
        }]
      }]
    }
    Flame2 {
     id=#7
     possessed_class=Image
     tracks [{
       type=MovieSceneWidgetMaterialTrack
       sections [{
         type=MovieSceneParameterSection
         channels [
          Scale_Anim {type=MovieSceneScriptingFloatChannel, keys [[0,1],[7,1],[8,0],[10,1]]}
          Offset_X {type=MovieSceneScriptingFloatChannel, const [0,3]}
          Offset_Y {type=MovieSceneScriptingFloatChannel, keys [[7,0],[8,0.15],[10,0]]}
          Final_Alpha {type=MovieSceneScriptingFloatChannel, keys [[0,0],[7,0],[8,1],[10,0]]}
         ]
        }]
      }]
    }
   ]}
 }]
graphs [EventGraph {nodes [
   n1 = Event OnCanCraftNowChanged!() -> (OutputDelegate, then, bCanCraftNow:bool) {is=title K2Node_Event}
   > SetActiveWidgetIndex(
    self=Get CraftingIndicatorSwitcher() -> (CraftingIndicatorSwitcher:Widget Switcher Object Reference)
    Index=Select(Option 0:int=0, Option 1:int=1, Index=@n1.bCanCraftNow) -> (ReturnValue:int)
   )
   > n2 = Branch(Condition=@n1.bCanCraftNow)
   > PlayAnimation(
    InAnimation=Get CanCraftNowAnim() -> (CanCraftNowAnim:Widget Animation Object Reference)
    NumLoopsToPlay=Get CanCraftAnimLoopCount() -> (CanCraftAnimLoopCount:int=0)
   )
   StopAnimation(execute=@n2.else, InAnimation=Get CanCraftNowAnim())
   n3 = Event OnIsCraftableItemChanged!() -> (OutputDelegate, then, bIsCraftableItem:bool) {is=title K2Node_Event}
   > SetVisibility(InVisibility=Select(Option 0:ESlateVisibility Enum=Collapsed, Option 1:ESlateVisibility Enum=SelfHitTestInvisible, Index=@n3.bIsCraftableItem) -> (ReturnValue:ESlateVisibility Enum))
   n4 = Event OnIngredientChanged!() -> (OutputDelegate, then, bCanCraftNow:bool) {is=title K2Node_Event}
   n5 = IsAnimationPlaying(InAnimation=Get CanCraftNowAnim())
   > Branch(Condition=AND Boolean(A=@n4.bCanCraftNow, B=NOT Boolean(A=@n5)))
   > PlayAnimation(InAnimation=Get CanCraftNowAnim(), NumLoopsToPlay=Get CanCraftAnimLoopCount() -> (CanCraftAnimLoopCount=0))
  ]}]
variables [CanCraftAnimLoopCount {type=integer, default=3}]
widgets {
 info {
  inheritedWidgetCount=0
  parentClass=/Script/CraftingUI.AthenaQuickBarSlotCraftingIndicator
  rootWidgetClass=&WidgetSwitcher
  widgetCount=11
 }
 widgets [
  CraftingIndicatorSwitcher {widgetClassPath=&WidgetSwitcher, bIsVariable=true, widget=CraftingIndicatorSwitcher, properties {activeWidgetIndex=1}}
  Null_Empty {widgetClassPath=&Spacer, parent=CraftingIndicatorSwitcher, widget=Null_Empty, slot {class=WidgetSwitcherSlot}}
  CraftingIndicator {
   widgetClassPath=&Overlay
   parent=CraftingIndicatorSwitcher
   widget=CraftingIndicator
   properties {renderTransform {angle=0, scale=<1 1>, shear=<0 0>, translation=<0 -24>}}
   slot {class=WidgetSwitcherSlot, properties {verticalAlignment=VAlign_Top}}
  }
  Image_558 {
   widgetClassPath=&Image
   parent=CraftingIndicator
   widget=Image_558
   properties {
    visibility=SelfHitTestInvisible
    brush {
     bIsDynamicallyLoaded=false
     drawAs=Image
     imageType=NoImage
     mirroring=NoMirror
     resourceName=None
     resourceObject=null
     tiling=NoTile
     imageSize=<32 14>
     margin=<ltrb 0 0 0 0>
     outlineSettings {
      bUseBrushTransparency=false
      roundingType=HalfHeightRadius
      width=0
      color {colorUseRule=UseColor_Specified, specifiedColor=<rgba 0 0 0 0>}
      cornerRadii=<0 0 0 0>
     }
     tintColor {colorUseRule=UseColor_Specified, specifiedColor=<rgba 1 1 1 1>}
     uVRegion {bIsValid=false, max=<0 0>, min=<0 0>}
    }
    colorAndOpacity=<rgba 0 0 0 0.5>
   }
   slot {class=OverlaySlot, properties {horizontalAlignment=HAlign_Fill, verticalAlignment=VAlign_Bottom}}
  }
  Backing_1 {
   widgetClassPath=&Image
   parent=CraftingIndicator
   widget=Backing_1
   properties {
    visibility=SelfHitTestInvisible
    brush {
     bIsDynamicallyLoaded=false
     drawAs=Image
     imageType=NoImage
     mirroring=NoMirror
     resourceName=None
     resourceObject=/Game/Athena/HUD/Inventory/Art/M_UI_ContextualCrafting_Interaction_Tint
     tiling=NoTile
     imageSize=<32 14>
     margin=<ltrb 0 0 0 0>
     outlineSettings {
      bUseBrushTransparency=false
      roundingType=HalfHeightRadius
      width=0
      color {colorUseRule=UseColor_Specified, specifiedColor=<rgba 0 0 0 0>}
      cornerRadii=<0 0 0 0>
     }
     tintColor {colorUseRule=UseColor_Specified, specifiedColor=<rgba 1 1 1 1>}
     uVRegion {bIsValid=false, max=<0 0>, min=<0 0>}
    }
   }
   slot {class=OverlaySlot, properties {horizontalAlignment=HAlign_Fill, verticalAlignment=VAlign_Bottom}}
  }
  Glow {
   widgetClassPath=&Image
   parent=CraftingIndicator
   widget=Glow
   properties {
    visibility=SelfHitTestInvisible
    brush {
     bIsDynamicallyLoaded=false
     drawAs=Image
     imageType=NoImage
     mirroring=NoMirror
     resourceName=None
     resourceObject=/Game/Athena/HUD/Inventory/Art/M_UI_CraftingIndicato_SequencerAnim_Glow
     tiling=NoTile
     imageSize=<24 24>
     margin=<ltrb 0 0 0 0>
     outlineSettings {
      bUseBrushTransparency=false
      roundingType=HalfHeightRadius
      width=0
      color {colorUseRule=UseColor_Specified, specifiedColor=<rgba 0 0 0 0>}
      cornerRadii=<0 0 0 0>
     }
     tintColor {colorUseRule=UseColor_Specified, specifiedColor=<rgba 1 1 1 1>}
     uVRegion {bIsValid=false, max=<0 0>, min=<0 0>}
    }
   }
   slot {class=OverlaySlot, properties {horizontalAlignment=HAlign_Fill, verticalAlignment=VAlign_Fill}}
  }
  Flame1 {
   widgetClassPath=&Image
   parent=CraftingIndicator
   widget=Flame1
   properties {
    visibility=SelfHitTestInvisible
    brush {
     bIsDynamicallyLoaded=false
     drawAs=Image
     imageType=NoImage
     mirroring=NoMirror
     resourceName=None
     resourceObject=/Game/Athena/HUD/Inventory/Art/M_UI_CraftingIndicato_SequencerAnim_Flame_1
     tiling=NoTile
     imageSize=<24 24>
     margin=<ltrb 0 0 0 0>
     outlineSettings {
      bUseBrushTransparency=false
      roundingType=HalfHeightRadius
      width=0
      color {colorUseRule=UseColor_Specified, specifiedColor=<rgba 0 0 0 0>}
      cornerRadii=<0 0 0 0>
     }
     tintColor {colorUseRule=UseColor_Specified, specifiedColor=<rgba 1 1 1 1>}
     uVRegion {bIsValid=false, max=<0 0>, min=<0 0>}
    }
   }
   slot {class=OverlaySlot, properties {horizontalAlignment=HAlign_Fill, verticalAlignment=VAlign_Fill}}
  }
  Flame2 {
   widgetClassPath=&Image
   parent=CraftingIndicator
   widget=Flame2
   properties {
    visibility=SelfHitTestInvisible
    brush {
     bIsDynamicallyLoaded=false
     drawAs=Image
     imageType=NoImage
     mirroring=NoMirror
     resourceName=None
     resourceObject=/Game/Athena/HUD/Inventory/Art/M_UI_CraftingIndicato_SequencerAnim_Flame_2
     tiling=NoTile
     imageSize=<24 24>
     margin=<ltrb 0 0 0 0>
     outlineSettings {
      bUseBrushTransparency=false
      roundingType=HalfHeightRadius
      width=0
      color {colorUseRule=UseColor_Specified, specifiedColor=<rgba 0 0 0 0>}
      cornerRadii=<0 0 0 0>
     }
     tintColor {colorUseRule=UseColor_Specified, specifiedColor=<rgba 1 1 1 1>}
     uVRegion {bIsValid=false, max=<0 0>, min=<0 0>}
    }
   }
   slot {class=OverlaySlot, properties {horizontalAlignment=HAlign_Fill, verticalAlignment=VAlign_Fill}}
  }
  Hammer {
   widgetClassPath=&Image
   parent=CraftingIndicator
   widget=Hammer
   properties {
    visibility=SelfHitTestInvisible
    brush {
     bIsDynamicallyLoaded=false
     drawAs=Image
     imageType=NoImage
     mirroring=NoMirror
     resourceName=None
     resourceObject=/Game/Athena/HUD/Inventory/Art/M_UI_CraftingIndicato_SequencerAnim_Hammer
     tiling=NoTile
     imageSize=<24 24>
     margin=<ltrb 0 0 0 0>
     outlineSettings {
      bUseBrushTransparency=false
      roundingType=HalfHeightRadius
      width=0
      color {colorUseRule=UseColor_Specified, specifiedColor=<rgba 0 0 0 0>}
      cornerRadii=<0 0 0 0>
     }
     tintColor {colorUseRule=UseColor_Specified, specifiedColor=<rgba 1 1 1 1>}
     uVRegion {bIsValid=false, max=<0 0>, min=<0 0>}
    }
   }
   slot {class=OverlaySlot, properties {horizontalAlignment=HAlign_Fill}}
  }
  Sparks1 {
   widgetClassPath=&Image
   parent=CraftingIndicator
   widget=Sparks1
   properties {
    visibility=SelfHitTestInvisible
    brush {
     bIsDynamicallyLoaded=false
     drawAs=Image
     imageType=NoImage
     mirroring=NoMirror
     resourceName=None
     resourceObject=/Game/Athena/HUD/Inventory/Art/M_UI_CraftingIndicato_SequencerAnim_Sparks1
     tiling=NoTile
     imageSize=<24 24>
     margin=<ltrb 0 0 0 0>
     outlineSettings {
      bUseBrushTransparency=false
      roundingType=HalfHeightRadius
      width=0
      color {colorUseRule=UseColor_Specified, specifiedColor=<rgba 0 0 0 0>}
      cornerRadii=<0 0 0 0>
     }
     tintColor {colorUseRule=UseColor_Specified, specifiedColor=<rgba 1 1 1 1>}
     uVRegion {bIsValid=false, max=<0 0>, min=<0 0>}
    }
    renderTransform {angle=0, scale=<1.5 1.5>, shear=<0 0>, translation=<0 0>}
   }
   slot {class=OverlaySlot, properties {horizontalAlignment=HAlign_Fill, verticalAlignment=VAlign_Fill}}
  }
  Sparks2 {
   widgetClassPath=&Image
   parent=CraftingIndicator
   widget=Sparks2
   properties {
    visibility=SelfHitTestInvisible
    brush {
     bIsDynamicallyLoaded=false
     drawAs=Image
     imageType=NoImage
     mirroring=NoMirror
     resourceName=None
     resourceObject=/Game/Athena/HUD/Inventory/Art/M_UI_CraftingIndicato_SequencerAnim_Sparks2
     tiling=NoTile
     imageSize=<24 24>
     margin=<ltrb 0 0 0 0>
     outlineSettings {
      bUseBrushTransparency=false
      roundingType=HalfHeightRadius
      width=0
      color {colorUseRule=UseColor_Specified, specifiedColor=<rgba 0 0 0 0>}
      cornerRadii=<0 0 0 0>
     }
     tintColor {colorUseRule=UseColor_Specified, specifiedColor=<rgba 1 1 1 1>}
     uVRegion {bIsValid=false, max=<0 0>, min=<0 0>}
    }
    renderTransform {angle=0, scale=<1.25 1.25>, shear=<0 0>, translation=<0 4>}
   }
   slot {class=OverlaySlot, properties {horizontalAlignment=HAlign_Fill, verticalAlignment=VAlign_Fill}}
  }
 ]
}
