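# Widget Blueprint for the crafting indicator in inventory item info, showing whether an item is craftable and whether it can be crafted now.
# The C++ parent's craftable-item event toggles the initially collapsed horizontal box between Collapsed and SelfHitTestInvisible.
# Its can-craft-now event sets the same index on icon and text switchers, swapping "Crafting Item" for "Crafting Ready!".
# The ready state uses a different icon material and a material-driven text outline.
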WidgetBlueprint /Crafting/UI/AthenaInventoryItemInfoCraftingIndicator
graphs [EventGraph {nodes [
   n1 = Event OnIsCraftableItemChanged!() -> (OutputDelegate, then, bIsCraftableItem:bool) {is=title K2Node_Event}
   > SetVisibility(
    self=Get HBoxContent() -> (HBoxContent:Horizontal Box Object Reference)
    InVisibility=Select(Option 0=Collapsed, Option 1=SelfHitTestInvisible, Index=@n1.bIsCraftableItem)
   )
   n2 = Event OnCanCraftNowChanged!() -> (OutputDelegate, then, bCanCraftNow:bool) {is=title K2Node_Event}
   n3 = Select(Option 0:int=0, Option 1:int=1, Index=@n2.bCanCraftNow) -> (ReturnValue:int)
   > SetActiveWidgetIndex(self=Get CraftingStateIcon() -> (CraftingStateIcon:Widget Switcher Object Reference), Index=@n3)
   > SetActiveWidgetIndex(self=Get CraftingStateText() -> (CraftingStateText:Widget Switcher Object Reference), Index=@n3)
  ]}]
variables []
widgets {
 info {
  inheritedWidgetCount=0
  parentClass=/Script/CraftingUI.AthenaInventoryItemInfoCraftingIndicator
  rootWidgetClass=&HorizontalBox
  widgetCount=7
 }
 widgets [
  HBoxContent {widgetClassPath=&HorizontalBox, bIsVariable=true, widget=HBoxContent, properties {visibility=Collapsed}}
  CraftingStateIcon {widgetClassPath=&WidgetSwitcher, bIsVariable=true, parent=HBoxContent, widget=CraftingStateIcon, slot {class=HorizontalBoxSlot}}
  NotReadyIcon {
   widgetClassPath=&Image
   parent=CraftingStateIcon
   widget=NotReadyIcon
   properties {
    visibility=SelfHitTestInvisible
    brush {
     bIsDynamicallyLoaded=false
     drawAs=Image
     imageType=NoImage
     mirroring=NoMirror
     resourceName=None
     resourceObject=/Game/Athena/HUD/Inventory/Art/M_UI_CraftingIndicator_Icon_Base
     tiling=NoTile
     imageSize=<24 24>
     margin=<ltrb 0 0 0 0>
     outlineSettings {
      bUseBrushTransparency=false
      roundingType=HalfHeightRadius
      width=0
      color {colorUseRule=UseColor_Specified, specifiedColor=<rgba 0 0 0 0>}
      cornerRadii=<0 0 0 0>
     }
     tintColor {colorUseRule=UseColor_Specified, specifiedColor=<rgba 1 1 1 1>}
     uVRegion {bIsValid=false, max=<0 0>, min=<0 0>}
    }
   }
   slot {class=WidgetSwitcherSlot}
  }
  ReadyIcon {
   widgetClassPath=&Image
   parent=CraftingStateIcon
   widget=ReadyIcon
   properties {
    visibility=SelfHitTestInvisible
    brush {
     bIsDynamicallyLoaded=false
     drawAs=Image
     imageType=NoImage
     mirroring=NoMirror
     resourceName=None
     resourceObject=/Game/Athena/HUD/Inventory/Art/M_UI_CraftingIndicator
     tiling=NoTile
     imageSize=<24 24>
     margin=<ltrb 0 0 0 0>
     outlineSettings {
      bUseBrushTransparency=false
      roundingType=HalfHeightRadius
      width=0
      color {colorUseRule=UseColor_Specified, specifiedColor=<rgba 0 0 0 0>}
      cornerRadii=<0 0 0 0>
     }
     tintColor {colorUseRule=UseColor_Specified, specifiedColor=<rgba 1 1 1 1>}
     uVRegion {bIsValid=false, max=<0 0>, min=<0 0>}
    }
    renderTransform {angle=0, scale=<1.125 1.125>, shear=<0 0>, translation=<0 0>}
   }
   slot {class=WidgetSwitcherSlot}
  }
  CraftingStateText {
   widgetClassPath=&WidgetSwitcher
   bIsVariable=true
   parent=HBoxContent
   widget=CraftingStateText
   slot {class=HorizontalBoxSlot, properties {verticalAlignment=VAlign_Center, padding=<ltrb 4 0 0 0>}}
  }
  NotReadyText {
   widgetClassPath=&CommonTextBlock
   parent=CraftingStateText
   widget=NotReadyText
   properties {
    justification=Right
    lineHeightPercentage=0.9
    mobileFontSizeMultiplier=1.2
    scrollStyle=/FortUILibrary/Styles/Typography/Foundation/TextScrollStyle-NoFade.TextScrollStyle-NoFade_C
    text=Crafting Item
    textTransformPolicy=ToUpper
    font {
     bForceMonospaced=false
     bMaterialIsStencil=false
     fontMaterial=/Game/Athena/HUD/Inventory/Art/M_UI_Inventory_Crafting_FontColor
     fontObject=&BurbankSmall
     letterSpacing=0
     monospacedWidth=1
     size=13
     skewAmount=0
     typefaceFontName=Black
     outlineSettings {
      bApplyOutlineToDropShadows=false
      bMiteredCorners=false
      bSeparateFillAlpha=false
      outlineMaterial=null
      outlineSize=0
      outlineColor=<rgba 0 0 0 1>
     }
    }
    margin=<ltrb 0 -2 0 0>
    renderTransform {angle=0, scale=<1 1>, shear=<-12 0>, translation=<0 0>}
    shadowOffset=<0 0>
   }
   slot {class=WidgetSwitcherSlot}
  }
  ReadyText {
   widgetClassPath=&CommonTextBlock
   parent=CraftingStateText
   widget=ReadyText
   properties {
    justification=Right
    lineHeightPercentage=0.9
    mobileFontSizeMultiplier=1.2
    scrollStyle=&TextScrollStyle-NoFade_C
    text="Crafting Ready!"
    textTransformPolicy=ToUpper
    font {
     bForceMonospaced=false
     bMaterialIsStencil=false
     fontMaterial=&M_UI_Inventory_Crafting_FontColor
     fontObject=&BurbankSmall
     letterSpacing=0
     monospacedWidth=1
     size=13
     skewAmount=0
     typefaceFontName=Black
     outlineSettings {
      bApplyOutlineToDropShadows=false
      bMiteredCorners=false
      bSeparateFillAlpha=false
      outlineMaterial=/Game/Athena/HUD/Inventory/Art/M_UI_Crafting_AlertState
      outlineSize=1
      outlineColor=<rgba 1 1 1 1>
     }
    }
    margin=<ltrb 0 -2 0 0>
    renderTransform {angle=0, scale=<1 1>, shear=<-12 0>, translation=<-1 0>}
    shadowOffset=<0 0>
   }
   slot {class=WidgetSwitcherSlot}
  }
 ]
}
