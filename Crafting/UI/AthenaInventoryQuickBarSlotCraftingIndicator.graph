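# Crafting indicator widget shown on an inventory quick bar slot, extending AthenaQuickBarSlotCraftingIndicator.
# Craftable-item changes show or collapse the indicator, while can-craft-now changes switch a visibility switcher between static and animated icon states and toggle the Bar image.
# Owner hover events shift the indicator upward, and selection slides the animated icon aside and pops in an InventoryTabRight input prompt.
# Deselect and deactivate reverse these animations from their end time.
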WidgetBlueprint /Crafting/UI/AthenaInventoryQuickBarSlotCraftingIndicator
animations [
 ShowCraftingPrompt {
  binding_count=2
  length_seconds=0.15
  object=&self:ShowCraftingPrompt
  sequence {bindings [
    AnimatedIcon {
     id=#1
     possessed_class=Image
     tracks [{
       type=MovieScene2DTransformTrack
       sections [{type=MovieScene2DTransformSection, channels [Translation.X {type=MovieSceneScriptingFloatChannel, keys [[0,0],[2.9997,14]]}]}]
      }]
    }
    CraftingPromptIndicator {
     id=#2
     possessed_class=ScaleBox
     tracks [
      {
       type=MovieSceneFloatTrack
       display_name=Render Opacity
       sections [{type=MovieSceneFloatSection, channels [None {type=MovieSceneScriptingFloatChannel, keys [[0,0],[2,1],[2.9997,1]]}]}]
      }
      {
       type=MovieScene2DTransformTrack
       sections [{
         type=MovieScene2DTransformSection
         channels [
          Translation.Y {type=MovieSceneScriptingFloatChannel, keys [[0,16],[2,-4],[2.9997,0]]}
          Scale.X {type=MovieSceneScriptingFloatChannel, keys [[0,0.75],[2,1.05],[2.9997,1]]}
          Scale.Y {type=MovieSceneScriptingFloatChannel, keys [[0,0.75],[2,1.05],[2.9997,1]]}
         ]
        }]
      }
     ]
    }
   ]}
 }
 Hover {
  binding_count=1
  length_seconds=0.1
  object=&self:Hover
  sequence {bindings [CraftingIndicator {
     id=#3
     possessed_class=Overlay
     tracks [{
       type=MovieScene2DTransformTrack
       sections [{type=MovieScene2DTransformSection, channels [Translation.Y {type=MovieSceneScriptingFloatChannel, keys [[0,-32],[1.9997,-36]]}]}]
      }]
    }]}
 }
]
graphs [EventGraph {nodes [
   Event OnOwnerHovered() {is=title K2Node_Event}
   > PlayAnimationForward(InAnimation=Get Hover() -> (Hover:Widget Animation Object Reference))
   > Set bIsOwnerHovered(bIsOwnerHovered:bool=true) -> (Output_Get:bool)
   Event OnOwnerUnhovered() {is=title K2Node_Event}
   > PlayAnimationReverse(InAnimation=Get Hover())
   > Set bIsOwnerHovered(bIsOwnerHovered=false)
   Event OnOwnerDeactivated() {is=title K2Node_Event}
   n1 = Get Hover()
   > PlayAnimation(InAnimation=@n1, StartAtTime=GetEndTime(self=@n1), PlayMode=Reverse)
   n2 = Get ShowCraftingPrompt() -> (ShowCraftingPrompt:Widget Animation Object Reference)
   > PlayAnimation(InAnimation=@n2, StartAtTime=GetEndTime(self=@n2), PlayMode=Reverse)
   > Set bIsOwnerSelected(bIsOwnerSelected:bool=false) -> (Output_Get:bool)
   > Set bIsOwnerHovered(bIsOwnerHovered=false)
   Event OnOwnerSelected() {is=title K2Node_Event}
   > PlayAnimationForward(InAnimation=Get ShowCraftingPrompt())
   > Set bIsOwnerSelected(bIsOwnerSelected=true)
   Event OnOwnerDeselected() {is=title K2Node_Event}
   n3 = Get ShowCraftingPrompt()
   > PlayAnimation(InAnimation=@n3, StartAtTime=GetEndTime(self=@n3), PlayMode=Reverse)
   > Set bIsOwnerSelected(bIsOwnerSelected=false)
   n4 = Event OnIsCraftableItemChanged!() -> (OutputDelegate, then, bIsCraftableItem:bool) {is=title K2Node_Event}
   > SetVisibility(
    self=Get CraftingIndicator() -> (CraftingIndicator:Overlay Object Reference)
    InVisibility=Select(Option 0=Collapsed, Option 1=SelfHitTestInvisible, Index=@n4.bIsCraftableItem)
   )
   n5 = Event OnCanCraftNowChanged!() -> (OutputDelegate, then, bCanCraftNow:bool) {is=title K2Node_Event}
   > SetActiveWidgetIndex(
    self:Common Visibility Switcher Object Reference=Get IsAnimatedIcon() -> (IsAnimatedIcon:Common Visibility Switcher Object Reference)
    Index=Select(Option 0:int=0, Option 1:int=1, Index=@n5.bCanCraftNow) -> (ReturnValue:int)
   )
   > SetVisibility(
    self=Get Bar() -> (Bar:Image Object Reference)
    InVisibility=Select(Option 0:ESlateVisibility Enum=Hidden, Option 1:ESlateVisibility Enum=SelfHitTestInvisible, Index=@n5.bCanCraftNow) -> (ReturnValue:ESlateVisibility Enum)
   )
  ]}]
variables [
 bIsOwnerHovered {type=boolean}
 bIsOwnerSelected {type=boolean}
]
widgets {
 info {
  inheritedWidgetCount=0
  parentClass=/Script/CraftingUI.AthenaQuickBarSlotCraftingIndicator
  rootWidgetClass=&Overlay
  widgetCount=11
 }
 widgets [
  Overlay_1 {widgetClassPath=&Overlay, widget=Overlay_1}
  CraftingIndicator {
   widgetClassPath=&Overlay
   bIsVariable=true
   parent=Overlay_1
   widget=CraftingIndicator
   properties {renderTransform {angle=0, scale=<1 1>, shear=<0 0>, translation=<0 -32>}}
   slot {class=OverlaySlot, properties {horizontalAlignment=HAlign_Fill}}
  }
  IsAnimatedIcon {
   widgetClassPath=&CommonVisibilitySwitcher
   bIsVariable=true
   parent=CraftingIndicator
   widget=IsAnimatedIcon
   properties {activeWidgetIndex=1}
   slot {class=OverlaySlot, properties {horizontalAlignment=HAlign_Fill, verticalAlignment=VAlign_Fill}}
  }
  NotAnimatedIcon {
   widgetClassPath=&Image
   parent=IsAnimatedIcon
   widget=NotAnimatedIcon
   properties {
    visibility=SelfHitTestInvisible
    brush {
     bIsDynamicallyLoaded=false
     drawAs=NoDrawType
     imageType=NoImage
     mirroring=NoMirror
     resourceName=None
     resourceObject=/Game/Athena/HUD/Inventory/Art/M_UI_CraftingIndicator_HUD
     tiling=NoTile
     imageSize=<32 32>
     margin=<ltrb 0 0 0 0>
     outlineSettings {
      bUseBrushTransparency=false
      roundingType=HalfHeightRadius
      width=0
      color {colorUseRule=UseColor_Specified, specifiedColor=<rgba 0 0 0 0>}
      cornerRadii=<0 0 0 0>
     }
     tintColor {colorUseRule=UseColor_Specified, specifiedColor=<rgba 1 1 1 1>}
     uVRegion {bIsValid=false, max=<0 0>, min=<0 0>}
    }
    renderTransform {angle=0, scale=<0.925 0.955>, shear=<0 0>, translation=<2 -0.6>}
    renderTransformPivot=<0.5 1>
   }
   slot {class=CommonVisibilitySwitcherSlot, properties {horizontalAlignment=HAlign_Center, verticalAlignment=VAlign_Bottom}}
  }
  AnimatedState {widgetClassPath=&Overlay, parent=IsAnimatedIcon, widget=AnimatedState, slot {class=CommonVisibilitySwitcherSlot}}
  DarkenBacking {
   widgetClassPath=&Image
   bIsVariable=true
   parent=AnimatedState
   widget=DarkenBacking
   properties {
    visibility=SelfHitTestInvisible
    brush {
     bIsDynamicallyLoaded=false
     drawAs=Image
     imageType=NoImage
     mirroring=NoMirror
     resourceName=None
     resourceObject=/Game/Athena/HUD/Inventory/Art/M_UI_HUD_CraftingIndicator_Backing_Dark
     tiling=NoTile
     imageSize=<32 20>
     margin=<ltrb 0 0 0 0>
     outlineSettings {
      bUseBrushTransparency=false
      roundingType=HalfHeightRadius
      width=0
      color {colorUseRule=UseColor_Specified, specifiedColor=<rgba 0 0 0 0>}
      cornerRadii=<0 0 0 0>
     }
     tintColor {colorUseRule=UseColor_Specified, specifiedColor=<rgba 1 1 1 1>}
     uVRegion {bIsValid=false, max=<0 0>, min=<0 0>}
    }
   }
   slot {class=OverlaySlot, properties {horizontalAlignment=HAlign_Fill, verticalAlignment=VAlign_Bottom}}
  }
  AnimatedBacking {
   widgetClassPath=&Image
   bIsVariable=true
   parent=AnimatedState
   widget=AnimatedBacking
   properties {brush {
     bIsDynamicallyLoaded=false
     drawAs=Image
     imageType=NoImage
     mirroring=NoMirror
     resourceName=None
     resourceObject=/Game/Athena/HUD/Inventory/Art/M_UI_Crafting_AlertState
     tiling=NoTile
     imageSize=<32 20>
     margin=<ltrb 0 0 0 0>
     outlineSettings {
      bUseBrushTransparency=false
      roundingType=HalfHeightRadius
      width=0
      color {colorUseRule=UseColor_Specified, specifiedColor=<rgba 0 0 0 0>}
      cornerRadii=<0 0 0 0>
     }
     tintColor {colorUseRule=UseColor_Specified, specifiedColor=<rgba 1 1 1 1>}
     uVRegion {bIsValid=false, max=<0 0>, min=<0 0>}
    }}
   slot {class=OverlaySlot, properties {horizontalAlignment=HAlign_Fill, verticalAlignment=VAlign_Bottom}}
  }
  AnimatedIcon {
   widgetClassPath=&Image
   parent=AnimatedState
   widget=AnimatedIcon
   properties {
    visibility=SelfHitTestInvisible
    brush {
     bIsDynamicallyLoaded=false
     drawAs=Image
     imageType=NoImage
     mirroring=NoMirror
     resourceName=None
     resourceObject=/Game/Athena/HUD/Inventory/Art/M_UI_CraftingIndicator
     tiling=NoTile
     imageSize=<32 32>
     margin=<ltrb 0 0 0 0>
     outlineSettings {
      bUseBrushTransparency=false
      roundingType=HalfHeightRadius
      width=0
      color {colorUseRule=UseColor_Specified, specifiedColor=<rgba 0 0 0 0>}
      cornerRadii=<0 0 0 0>
     }
     tintColor {colorUseRule=UseColor_Specified, specifiedColor=<rgba 1 1 1 1>}
     uVRegion {bIsValid=false, max=<0 0>, min=<0 0>}
    }
    renderTransform {angle=0, scale=<1 1>, shear=<0 0>, translation=<0 2>}
   }
   slot {class=OverlaySlot, properties {horizontalAlignment=HAlign_Fill}}
  }
  CraftingPromptIndicator {
   widgetClassPath=&ScaleBox
   bIsVariable=true
   parent=AnimatedState
   widget=CraftingPromptIndicator
   properties {renderOpacity=0, stretch=UserSpecified, userSpecifiedScale=0.7}
   slot {class=OverlaySlot, properties {verticalAlignment=VAlign_Fill, padding=<ltrb 34 0 0 0>}}
  }
  CommonActionWidget_190 {
   widgetClassPath=&CommonActionWidget
   parent=CraftingPromptIndicator
   widget=CommonActionWidget_190
   properties {inputActions [{dataTable=/FortUILibrary/Input/AthenaInventoryActionDataTable, rowName=InventoryTabRight}]}
   slot {class=ScaleBoxSlot}
  }
  Bar {
   widgetClassPath=&Image
   bIsVariable=true
   parent=CraftingIndicator
   widget=Bar
   properties {
    visibility=SelfHitTestInvisible
    brush {
     bIsDynamicallyLoaded=false
     drawAs=NoDrawType
     imageType=NoImage
     mirroring=NoMirror
     resourceName=None
     resourceObject=/Game/Athena/HUD/Inventory/Art/M_UI_HUD_EquippedItem_CraftableIndicator_Backing
     tiling=NoTile
     imageSize=<32 8>
     margin=<ltrb 0 0 0 0>
     outlineSettings {
      bUseBrushTransparency=false
      roundingType=HalfHeightRadius
      width=0
      color {colorUseRule=UseColor_Specified, specifiedColor=<rgba 0 0 0 0>}
      cornerRadii=<0 0 0 0>
     }
     tintColor {colorUseRule=UseColor_Specified, specifiedColor=<rgba 1 1 1 1>}
     uVRegion {bIsValid=false, max=<0 0>, min=<0 0>}
    }
   }
   slot {class=OverlaySlot, properties {horizontalAlignment=HAlign_Fill, verticalAlignment=VAlign_Bottom}}
  }
 ]
}
