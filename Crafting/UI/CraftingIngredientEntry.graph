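# Crafting ingredient entry widget showing an item icon, available/required counts and an "And" plus separator, which is hidden on the last ingredient.
# OnIngredientWidgetUpdated compares available to required, then sets the count color, ReadyBorder material color and check/X switcher.
# Dynamic material parameters flag the icon as recipe view on construct and as an ingredient on update.
# Logic extends FortCraftingIngredientWidget; the primary-ingredient flag has no effect here.
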WidgetBlueprint /Crafting/UI/CraftingIngredientEntry
graphs [EventGraph {nodes [
   n1 = Event OnIngredientWidgetUpdated!() -> (OutputDelegate, then, NumAvailable:int=0, NumRequired:int=0, bIsPrimaryIngredient:bool, bIsLastIngredient:bool) {is=title K2Node_Event}
   n2 = "integer >= integer"!(A:int=@n1.NumAvailable, B:int=@n1.NumRequired) -> (ReturnValue:bool) {is=title K2Node_CallFunction}
   > SetColorAndOpacity(
    self=Get Text_NumAvailable() -> (Text_NumAvailable:Common Text Object Reference)
    InColorAndOpacity=Select(
     Option 0:Slate Color Structure=Get NeedsNotMet() -> (NeedsNotMet:Slate Color Structure)
     Option 1:Slate Color Structure=Get NeedsMet() -> (NeedsMet:Slate Color Structure)
     Index=@n2
    ) -> (ReturnValue:Slate Color Structure)
   )
   > n3 = GetDynamicMaterial(self=Get ReadyBorder() -> (ReadyBorder:Image Object Reference))
   > SetVectorParameterValue(
    self=@n3
    ParameterName=Outer_Color
    Value=Select(
     Option 0:Linear Color Structure=Get NeedsNotMet_PlusColor() -> (NeedsNotMet_PlusColor:Linear Color Structure)
     Option 1:Linear Color Structure=Get NeedsMet_PlusColor() -> (NeedsMet_PlusColor:Linear Color Structure)
     Index=@n2
    ) -> (ReturnValue:Linear Color Structure)
   )
   > SetVisibility(
    self=Get PlusSignColor() -> (PlusSignColor:Border Object Reference)
    InVisibility=Select(Option 0:ESlateVisibility Enum=SelfHitTestInvisible, Option 1:ESlateVisibility Enum=Collapsed, Index=@n1.bIsLastIngredient) -> (ReturnValue:ESlateVisibility Enum)
   )
   > SetActiveWidgetIndex(
    self=Get NeedsMetIndicator() -> (NeedsMetIndicator:Common Widget Switcher Legacy Object Reference)
    Index=Select(Option 0:int=1, Option 1:int=0, Index=@n2) -> (ReturnValue:int)
   )
   > n4 = Cast To AthenaItemIconLoadout!(execute, Object=Get ItemIcon() -> (ItemIcon:Athena Item Icon Object Reference)) -> (then, CastFailed, AsAthena Item Icon Loadout:Athena Item Icon Loadout Object Reference) {is=title K2Node_DynamicCast}
   > n5 = GetDynamicMaterial(self=Get Image_ItemIcon!(self:Athena Item Icon Object Reference=@n4) -> (Image_ItemIcon:Image Object Reference))
   > SetScalarParameterValue(self=@n5, ParameterName=IsIngredient, Value=1)
   n6 = Event PreConstruct()
   n7 = Cast To AthenaItemIconLoadout(execute=[@n6.then, @n9.then], Object=Get ItemIcon())
   > n8 = GetDynamicMaterial(self=Get Image_ItemIcon(self=@n7))
   > SetScalarParameterValue(self=@n8, ParameterName=IsRecipeView, Value=1)
   n9 = Event Construct()
   Reroute Node(InputPin:bool=@n1.bIsPrimaryIngredient) -> (OutputPin:bool)
  ]}]
variables [
 NeedsMet {type=SlateColor, default {colorUseRule=UseColor_Specified, specifiedColor=<rgba 1 1 1 1>}}
 NeedsMet_PlusColor {type=LinearColor, default=<rgba 0 0.7529 0.2016 1>}
 NeedsNotMet {type=SlateColor, default {colorUseRule=UseColor_Specified, specifiedColor=<rgba 0.7529 0.011 0.0482 1>}}
 NeedsNotMet_PlusColor {type=LinearColor, default=<rgba 0.7529 0.011 0.0482 1>}
]
widgets {
 info {
  inheritedWidgetCount=4
  parentClass=/Script/CraftingUI.FortCraftingIngredientWidget
  rootWidgetClass=&HorizontalBox
  widgetCount=19
 }
 widgets [
  HorizontalBox_1 {widgetClassPath=&HorizontalBox, widget=HorizontalBox_1}
  SizeBox_2 {
   widgetClassPath=&SizeBox
   parent=HorizontalBox_1
   widget=SizeBox_2
   properties {widthOverride=102}
   slot {class=HorizontalBoxSlot}
  }
  VerticalBox_0 {widgetClassPath=&VerticalBox, parent=SizeBox_2, widget=VerticalBox_0, slot {class=SizeBoxSlot}}
  SizeBox_0 {
   widgetClassPath=&SizeBox
   parent=VerticalBox_0
   widget=SizeBox_0
   properties {bOverride_HeightOverride=true, bOverride_WidthOverride=true, heightOverride=80, widthOverride=80}
   slot {class=VerticalBoxSlot, properties {horizontalAlignment=HAlign_Center, verticalAlignment=VAlign_Center}}
  }
  Overlay_1 {widgetClassPath=&Overlay, parent=SizeBox_0, widget=Overlay_1, slot {class=SizeBoxSlot}}
  ReadyBorder {
   widgetClassPath=&Image
   bIsVariable=true
   parent=Overlay_1
   widget=ReadyBorder
   properties {
    visibility=SelfHitTestInvisible
    brush {
     bIsDynamicallyLoaded=false
     drawAs=Box
     imageType=NoImage
     mirroring=NoMirror
     resourceName=None
     resourceObject=/Game/Athena/HUD/Inventory/Art/M_UI_HUD_Crafting_Ingredient_ReadyBorder
     tiling=NoTile
     imageSize=<8 8>
     margin=<ltrb 0.5 0.5 0.5 0.5>
     outlineSettings {
      bUseBrushTransparency=false
      roundingType=HalfHeightRadius
      width=0
      color {colorUseRule=UseColor_Specified, specifiedColor=<rgba 0 0 0 0>}
      cornerRadii=<0 0 0 0>
     }
     tintColor {colorUseRule=UseColor_Specified, specifiedColor=<rgba 1 1 1 1>}
     uVRegion {bIsValid=false, max=<0 0>, min=<0 0>}
    }
    renderTransform {angle=0, scale=<1.125 1.125>, shear=<0 0>, translation=<0 0>}
   }
   slot {class=OverlaySlot, properties {horizontalAlignment=HAlign_Fill, verticalAlignment=VAlign_Fill}}
  }
  LazyImage_Icon {
   widgetClassPath=&FortLazyImage
   bInherited=true
   parent=Overlay_1
   widget=LazyImage_Icon
   properties {brush {
     bIsDynamicallyLoaded=false
     drawAs=Image
     imageType=NoImage
     mirroring=NoMirror
     resourceName=None
     resourceObject=null
     tiling=NoTile
     imageSize=<48 48>
     margin=<ltrb 0 0 0 0>
     outlineSettings {
      bUseBrushTransparency=false
      roundingType=HalfHeightRadius
      width=0
      color {colorUseRule=UseColor_Specified, specifiedColor=<rgba 0 0 0 0>}
      cornerRadii=<0 0 0 0>
     }
     tintColor {colorUseRule=UseColor_Specified, specifiedColor=<rgba 1 1 1 1>}
     uVRegion {bIsValid=false, max=<0 0>, min=<0 0>}
    }}
   slot {class=OverlaySlot, properties {horizontalAlignment=HAlign_Center, verticalAlignment=VAlign_Center}}
  }
  ItemIcon {
   widgetClassPath=/Game/Athena/HUD/Inventory/AthenaItemIconLoadout
   bInherited=true
   parent=Overlay_1
   widget=ItemIcon
   properties {bShowAmmoIcon=false, bShowStackCountText=false}
   slot {class=OverlaySlot, properties {horizontalAlignment=HAlign_Fill, verticalAlignment=VAlign_Fill}}
  }
  NeedsMetIndicator {
   widgetClassPath=&CommonWidgetSwitcherLegacy
   bIsVariable=true
   parent=Overlay_1
   widget=NeedsMetIndicator
   properties {transitionCurveType=Linear, transitionDuration=0, renderTransform {angle=0, scale=<1 1>, shear=<0 0>, translation=<16 -16>}}
   slot {class=OverlaySlot, properties {horizontalAlignment=HAlign_Right}}
  }
  Check {
   widgetClassPath=&Image
   parent=NeedsMetIndicator
   widget=Check
   properties {
    visibility=SelfHitTestInvisible
    brush {
     bIsDynamicallyLoaded=false
     drawAs=Image
     imageType=NoImage
     mirroring=NoMirror
     resourceName=None
     resourceObject=/FortUILibrary/Foundation/Art_ToolKit/AngledBoxes/M_UI_AngledCheckBox_Crafting
     tiling=NoTile
     imageSize=<26 26>
     margin=<ltrb 0 0 0 0>
     outlineSettings {
      bUseBrushTransparency=false
      roundingType=HalfHeightRadius
      width=0
      color {colorUseRule=UseColor_Specified, specifiedColor=<rgba 0 0 0 0>}
      cornerRadii=<0 0 0 0>
     }
     tintColor {colorUseRule=UseColor_Specified, specifiedColor=<rgba 1 1 1 1>}
     uVRegion {bIsValid=false, max=<0 0>, min=<0 0>}
    }
   }
   slot {class=WidgetSwitcherSlot}
  }
  X {
   widgetClassPath=&Image
   parent=NeedsMetIndicator
   widget=X
   properties {
    visibility=SelfHitTestInvisible
    brush {
     bIsDynamicallyLoaded=false
     drawAs=Image
     imageType=NoImage
     mirroring=NoMirror
     resourceName=None
     resourceObject=/FortUILibrary/Foundation/Art_ToolKit/AngledBoxes/M_UI_AngledFailBox_Crafting
     tiling=NoTile
     imageSize=<26 26>
     margin=<ltrb 0 0 0 0>
     outlineSettings {
      bUseBrushTransparency=false
      roundingType=HalfHeightRadius
      width=0
      color {colorUseRule=UseColor_Specified, specifiedColor=<rgba 0 0 0 0>}
      cornerRadii=<0 0 0 0>
     }
     tintColor {colorUseRule=UseColor_Specified, specifiedColor=<rgba 1 1 1 1>}
     uVRegion {bIsValid=false, max=<0 0>, min=<0 0>}
    }
   }
   slot {class=WidgetSwitcherSlot}
  }
  HBox_Quantities {
   widgetClassPath=&HorizontalBox
   bIsVariable=true
   parent=VerticalBox_0
   widget=HBox_Quantities
   slot {class=VerticalBoxSlot, properties {horizontalAlignment=HAlign_Center, padding=<ltrb 0 4 0 0>}}
  }
  Text_NumAvailable {
   widgetClassPath=&CommonTextBlock
   bInherited=true
   parent=HBox_Quantities
   widget=Text_NumAvailable
   properties {
    mobileFontSizeMultiplier=1.4
    text=888
    font {
     bForceMonospaced=false
     bMaterialIsStencil=false
     fontMaterial=null
     fontObject=&BurbankSmall
     letterSpacing=0
     monospacedWidth=1
     size=18
     skewAmount=0
     typefaceFontName=Black
     outlineSettings {
      bApplyOutlineToDropShadows=false
      bMiteredCorners=false
      bSeparateFillAlpha=false
      outlineMaterial=null
      outlineSize=0
      outlineColor=<rgba 0 0 0 1>
     }
    }
    shadowOffset=<0 0>
   }
   slot {class=HorizontalBoxSlot, properties {verticalAlignment=VAlign_Center}}
  }
  Separator {
   widgetClassPath=&CommonTextBlock
   parent=HBox_Quantities
   widget=Separator
   properties {
    text=/
    font {
     bForceMonospaced=false
     bMaterialIsStencil=false
     fontMaterial=/Game/Athena/HUD/Inventory/Art/M_UI_Inventory_ItemInfo_Description_FontColor
     fontObject=&BurbankSmall
     letterSpacing=0
     monospacedWidth=1
     size=15
     skewAmount=0
     typefaceFontName=Bold
     outlineSettings {
      bApplyOutlineToDropShadows=false
      bMiteredCorners=false
      bSeparateFillAlpha=false
      outlineMaterial=null
      outlineSize=0
      outlineColor=<rgba 0 0 0 1>
     }
    }
    shadowOffset=<0 0>
   }
   slot {class=HorizontalBoxSlot, properties {verticalAlignment=VAlign_Center, padding=<ltrb 2 0 0 0>}}
  }
  Text_NumRequired {
   widgetClassPath=&CommonTextBlock
   bInherited=true
   parent=HBox_Quantities
   widget=Text_NumRequired
   properties {
    mobileFontSizeMultiplier=1.4
    text=888
    font {
     bForceMonospaced=false
     bMaterialIsStencil=false
     fontMaterial=&M_UI_Inventory_ItemInfo_Description_FontColor
     fontObject=&BurbankSmall
     letterSpacing=0
     monospacedWidth=1
     size=15
     skewAmount=0
     typefaceFontName=Bold
     outlineSettings {
      bApplyOutlineToDropShadows=false
      bMiteredCorners=false
      bSeparateFillAlpha=false
      outlineMaterial=null
      outlineSize=0
      outlineColor=<rgba 0 0 0 1>
     }
    }
    shadowOffset=<0 0>
   }
   slot {class=HorizontalBoxSlot, properties {verticalAlignment=VAlign_Center, padding=<ltrb 2 0 0 0>}}
  }
  PlusSignColor {
   widgetClassPath=&Border
   bIsVariable=true
   parent=HorizontalBox_1
   widget=PlusSignColor
   properties {
    visibility=SelfHitTestInvisible
    background {
     bIsDynamicallyLoaded=false
     drawAs=NoDrawType
     imageType=NoImage
     mirroring=NoMirror
     resourceName=None
     resourceObject=null
     tiling=NoTile
     imageSize=<32 32>
     margin=<ltrb 0 0 0 0>
     outlineSettings {
      bUseBrushTransparency=false
      roundingType=HalfHeightRadius
      width=0
      color {colorUseRule=UseColor_Specified, specifiedColor=<rgba 0 0 0 0>}
      cornerRadii=<0 0 0 0>
     }
     tintColor {colorUseRule=UseColor_Specified, specifiedColor=<rgba 1 1 1 1>}
     uVRegion {bIsValid=false, max=<0 0>, min=<0 0>}
    }
    contentColorAndOpacity=<rgba 0.7529 0.7529 0.7529 1>
    padding=<ltrb 0 0 0 0>
   }
   slot {class=HorizontalBoxSlot, properties {verticalAlignment=VAlign_Top, padding=<ltrb 20 12 20 0>}}
  }
  Overlay_0 {widgetClassPath=&Overlay, parent=PlusSignColor, widget=Overlay_0, slot {class=BorderSlot, properties {padding=<ltrb 0 0 0 0>}}}
  PlusSign {
   widgetClassPath=&Image
   parent=Overlay_0
   widget=PlusSign
   properties {
    renderOpacity=0.125
    visibility=SelfHitTestInvisible
    brush {
     bIsDynamicallyLoaded=false
     drawAs=Image
     imageType=NoImage
     mirroring=NoMirror
     resourceName=None
     resourceObject=/Game/Athena/UI/Frontend/Parties/Art/T_UI_PlusSign_64x
     tiling=NoTile
     imageSize=<56 56>
     margin=<ltrb 0 0 0 0>
     outlineSettings {
      bUseBrushTransparency=false
      roundingType=HalfHeightRadius
      width=0
      color {colorUseRule=UseColor_Specified, specifiedColor=<rgba 0 0 0 0>}
      cornerRadii=<0 0 0 0>
     }
     tintColor {colorUseRule=UseColor_Specified, specifiedColor=<rgba 1 1 1 1>}
     uVRegion {bIsValid=false, max=<0 0>, min=<0 0>}
    }
    renderTransform {angle=0, scale=<1 1>, shear=<-12 0>, translation=<0 0>}
   }
   slot {class=OverlaySlot}
  }
  CommonTextBlock_146 {
   widgetClassPath=&CommonTextBlock
   parent=Overlay_0
   widget=CommonTextBlock_146
   properties {
    justification=Center
    lineHeightPercentage=0.9
    mobileFontSizeMultiplier=1.5
    renderOpacity=0.5
    text=And
    textTransformPolicy=ToUpper
    font {
     bForceMonospaced=false
     bMaterialIsStencil=false
     fontMaterial=null
     fontObject=&BurbankSmall
     letterSpacing=0
     monospacedWidth=1
     size=13
     skewAmount=0
     typefaceFontName=Black
     outlineSettings {
      bApplyOutlineToDropShadows=false
      bMiteredCorners=false
      bSeparateFillAlpha=false
      outlineMaterial=null
      outlineSize=0
      outlineColor=<rgba 0 0 0 1>
     }
    }
    margin=<ltrb 0 -2 0 0>
    renderTransform {angle=0, scale=<1 1>, shear=<-16 0>, translation=<0 0>}
    shadowOffset=<0 0>
   }
   slot {class=OverlaySlot, properties {horizontalAlignment=HAlign_Center, verticalAlignment=VAlign_Center, padding=<ltrb 0 8 0 0>}}
  }
 ]
}
