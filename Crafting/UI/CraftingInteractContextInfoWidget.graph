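# Interact-prompt widget that previews the item craftable from the targeted pickup, with an icon and action text.
# On context update, it reads the last target formula from the player's crafting pickup-override component and fetches the first result.
# It async-loads that item's icon into a dynamic material with rarity background colors.
# Default text covers invalid or empty cases, and also overwrites the crafting subtext at the end of the success path.
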WidgetBlueprint /Crafting/UI/CraftingInteractContextInfoWidget
graphs [
 EventGraph {nodes [
   n1 = Event OnContextInfoUpdated!() -> (OutputDelegate, then, ContextInfo:Fort Interact Context Info Object Reference) {is=title K2Node_Event}
   > n2 = Is Valid(InputObject=@n1.ContextInfo)
   n3 = GetComponentByClass!(
    self:Actor Object Reference=GetOwningPlayer()
    ComponentClass:Actor Component Class Reference=/Script/CraftingRuntime.FortPickupInteractOverrideComponent_Crafting
   ) -> (ReturnValue:Fort Pickup Interact Override Component Crafting Object Reference)
   > n4 = Is Valid(InputObject=@n3)
   > n5 = GetCraftingResults!(
    execute
    WorldContextObject:Object Reference=@n3
    CraftingFormulaRow:"Name (by ref)"=Get LastTargetFormulaName!(self:Fort Pickup Interact Override Component Crafting Object Reference=@n3) -> (LastTargetFormulaName:Name=None)
    NumToCraft:int=1
   ) -> (then, OutResults:Array of Item And Count Structures)
   > n6 = Branch(Condition="integer > integer"(A=Length(TargetArray:Array of Item And Count Structures=@n5), B=0) {is=title K2Node_CallFunction})
   n7 = "Get (a ref)"(Array:Array of Item And Count Structures=@n5, Dimension 1=0) -> (Output:"Item And Count Structure (by ref)")
   > n8 = Async Load Asset(Asset=GetSmallPreviewImage!(ItemDefinition:Item Definition Base Object Reference=Break Item and Count!(ItemAndCount:"Item And Count Structure (by ref)"=@n7) -> (
      Count:int
      Item:Item Definition Base Object Reference
      ItemVariantGuid:Guid Structure
      WeaponModSlots:Array of Fort Weapon Mod Slot Structures
     ) {is=title K2Node_BreakStruct}.Item) -> (ReturnValue:Texture 2D Soft Object Reference))
   n9 = Cast To Texture2D(execute=@n8.Completed, Object=@n8)
   > SetCraftingResult!(
    execute
    self
    ContextInfo:"Fort Interact Context Info Object Reference (by ref)"=@n1.ContextInfo
    CraftingResult:"Item And Count Structure (by ref)"=@n7
   ) -> (then)
   > SetCraftingPreviewImage!(execute, self, ItemAndCount:"Item And Count Structure (by ref)"=@n7, ItemIcon:Texture Object Reference=@n9) -> (then)
   > SetDefaultText!(execute, self, ContextInfo:"Fort Interact Context Info Object Reference (by ref)"=@n1.ContextInfo) -> (then)
   SetDefaultText(execute=[@n6.else, @n4.Is Not Valid, @n2.Is Not Valid], ContextInfo=@n1.ContextInfo)
   Event PreConstruct()
   Event Construct()
   Event Tick()
  ]}
 SetCraftingPreviewImage {nodes [
   n1 = SetCraftingPreviewImage!() -> (then, ItemAndCount:"Item And Count Structure (by ref)", ItemIcon:Texture Object Reference) {is=title K2Node_FunctionEntry}
   > n2 = GetRarityDataForItemDef!(execute, ItemDef:Item Definition Base Object Reference=Break Item and Count(ItemAndCount=@n1).Item) -> (then, ReturnValue:Fort Rarity Item Data Structure)
   > n3 = GetDynamicMaterial(self=Get ItemPreviewImage() -> (ItemPreviewImage:Image Object Reference))
   > SetTextureParameterValue(self=@n3, ParameterName=ItemIcon, Value=@n1.ItemIcon)
   n4 = Break Fort Rarity Item Data!(FortRarityItemData:"Fort Rarity Item Data Structure (by ref)"=@n2) -> (
    Name:Text
    Color1:Linear Color Structure
    Color2:Linear Color Structure
    Color3:Linear Color Structure
    Color4:Linear Color Structure
    Color5:Linear Color Structure
    Radius:float=~
    Falloff:float=~
    Brightness:float=~
    Roughness:float=~
    Glow:float=~
   ) {is=title K2Node_BreakStruct}
   > SetVectorParameterValue(self=@n3, ParameterName=BG_Hotspot_Color, Value=@n4.Color2)
   > SetVectorParameterValue(self=@n3, ParameterName=BG_Falloff_Color, Value=@n4.Color3)
  ]}
 SetCraftingResult {nodes [
   n1 = SetCraftingResult!() -> (then, ContextInfo:"Fort Interact Context Info Object Reference (by ref)", CraftingResult:"Item And Count Structure (by ref)") {is=title K2Node_FunctionEntry}
   > "SetText (Text)"(
    self=Get ActionText() -> (ActionText:Common Text Object Reference)
    InText=Get SecondContextSubText!(self:Fort Interact Context Info Object Reference=@n1) -> (SecondContextSubText:Text)
   )
   Reroute Node(InputPin:Wildcard) -> (OutputPin:Wildcard)
   Is Valid(InputObject=Break Item and Count(ItemAndCount=@n1.CraftingResult).Item)
  ]}
 SetDefaultText {nodes [
   n1 = SetDefaultText!() -> (then, ContextInfo:"Fort Interact Context Info Object Reference (by ref)") {is=title K2Node_FunctionEntry}
   > "SetText (Text)"(self=Get ActionText(), InText=Get SecondContextText!(self:Fort Interact Context Info Object Reference=@n1) -> (SecondContextText:Text))
  ]}
]
variables [Image {type=MaterialExpressionTextureObjectParameter, default=null}]
widgets {
 info {inheritedWidgetCount=0, parentClass=/Script/FortniteGame.FortInteractContextInfoWidget, rootWidgetClass=&Border, widgetCount=6}
 widgets [
  Border_0 {
   widgetClassPath=&Border
   widget=Border_0
   properties {
    visibility=SelfHitTestInvisible
    background {
     bIsDynamicallyLoaded=false
     drawAs=Box
     imageType=NoImage
     mirroring=NoMirror
     resourceName=None
     resourceObject=/Game/Athena/HUD/Inventory/Art/M_UI_HUD_ItemDetails_CraftingBacking
     tiling=NoTile
     imageSize=<8 8>
     margin=<ltrb 0.5 0.5 0.5 0.5>
     outlineSettings {
      bUseBrushTransparency=false
      roundingType=HalfHeightRadius
      width=0
      color {colorUseRule=UseColor_Specified, specifiedColor=<rgba 0 0 0 0>}
      cornerRadii=<0 0 0 0>
     }
     tintColor {colorUseRule=UseColor_Specified, specifiedColor=<rgba 1 1 1 1>}
     uVRegion {bIsValid=false, max=<0 0>, min=<0 0>}
    }
    padding=<ltrb 8 8 12 8>
   }
  }
  HorizontalBox_0 {
   widgetClassPath=&HorizontalBox
   parent=Border_0
   widget=HorizontalBox_0
   slot {class=BorderSlot, properties {padding=<ltrb 8 8 12 8>}}
  }
  ItemPreviewImage {
   widgetClassPath=&Image
   bIsVariable=true
   parent=HorizontalBox_0
   widget=ItemPreviewImage
   properties {
    visibility=SelfHitTestInvisible
    brush {
     bIsDynamicallyLoaded=false
     drawAs=Image
     imageType=NoImage
     mirroring=NoMirror
     resourceName=None
     resourceObject=/Game/Athena/HUD/Inventory/Art/M_UI_HUD_InGame_CraftingPreview
     tiling=NoTile
     imageSize=<56 56>
     margin=<ltrb 0 0 0 0>
     outlineSettings {
      bUseBrushTransparency=false
      roundingType=HalfHeightRadius
      width=0
      color {colorUseRule=UseColor_Specified, specifiedColor=<rgba 0 0 0 0>}
      cornerRadii=<0 0 0 0>
     }
     tintColor {colorUseRule=UseColor_Specified, specifiedColor=<rgba 1 1 1 1>}
     uVRegion {bIsValid=false, max=<0 0>, min=<0 0>}
    }
   }
   slot {class=HorizontalBoxSlot}
  }
  Image_209 {
   widgetClassPath=&Image
   parent=HorizontalBox_0
   widget=Image_209
   properties {
    visibility=Collapsed
    brush {
     bIsDynamicallyLoaded=false
     drawAs=Image
     imageType=NoImage
     mirroring=NoMirror
     resourceName=None
     resourceObject=/Game/Athena/HUD/Inventory/Art/M_UI_HUD_ItemDetails_CraftingArrow_Base
     tiling=NoTile
     imageSize=<12 12>
     margin=<ltrb 0 0 0 0>
     outlineSettings {
      bUseBrushTransparency=false
      roundingType=HalfHeightRadius
      width=0
      color {colorUseRule=UseColor_Specified, specifiedColor=<rgba 0 0 0 0>}
      cornerRadii=<0 0 0 0>
     }
     tintColor {colorUseRule=UseColor_Specified, specifiedColor=<rgba 1 1 1 1>}
     uVRegion {bIsValid=false, max=<0 0>, min=<0 0>}
    }
   }
   slot {class=HorizontalBoxSlot, properties {horizontalAlignment=HAlign_Center, verticalAlignment=VAlign_Center, padding=<ltrb 6 0 0 0>}}
  }
  Image_161 {
   widgetClassPath=&Image
   parent=HorizontalBox_0
   widget=Image_161
   properties {
    visibility=SelfHitTestInvisible
    brush {
     bIsDynamicallyLoaded=false
     drawAs=Image
     imageType=NoImage
     mirroring=NoMirror
     resourceName=None
     resourceObject=/Game/Athena/HUD/Inventory/Art/M_UI_CraftingIndicator_Icon_White
     tiling=NoTile
     imageSize=<24 24>
     margin=<ltrb 0 0 0 0>
     outlineSettings {
      bUseBrushTransparency=false
      roundingType=HalfHeightRadius
      width=0
      color {colorUseRule=UseColor_Specified, specifiedColor=<rgba 0 0 0 0>}
      cornerRadii=<0 0 0 0>
     }
     tintColor {colorUseRule=UseColor_Specified, specifiedColor=<rgba 1 1 1 1>}
     uVRegion {bIsValid=false, max=<0 0>, min=<0 0>}
    }
   }
   slot {class=HorizontalBoxSlot, properties {horizontalAlignment=HAlign_Center, verticalAlignment=VAlign_Center, padding=<ltrb 6 0 0 0>}}
  }
  ActionText {
   widgetClassPath=&CommonTextBlock
   bIsVariable=true
   parent=HorizontalBox_0
   widget=ActionText
   properties {
    lineHeightPercentage=0.9
    mobileFontSizeMultiplier=1.3
    text="{Action}"
    textTransformPolicy=ToUpper
    font {
     bForceMonospaced=false
     bMaterialIsStencil=false
     fontMaterial=/Game/Athena/HUD/Inventory/Art/M_UI_ItemDetails_CraftingText_FontColor
     fontObject=&BurbankSmall
     letterSpacing=0
     monospacedWidth=1
     size=13
     skewAmount=0
     typefaceFontName=Black
     outlineSettings {
      bApplyOutlineToDropShadows=false
      bMiteredCorners=false
      bSeparateFillAlpha=false
      outlineMaterial=&M-FontShadow
      outlineSize=2
      outlineColor=<rgba 0.0625 0.0167 0.0002 0.5>
     }
    }
    margin=<ltrb 0 -1 0 0>
    renderTransform {angle=0, scale=<1 1>, shear=<-12 0>, translation=<0 0>}
    shadowOffset=<0 0>
   }
   slot {class=HorizontalBoxSlot, properties {verticalAlignment=VAlign_Center, padding=<ltrb 6 0 0 0>}}
  }
 ]
}
