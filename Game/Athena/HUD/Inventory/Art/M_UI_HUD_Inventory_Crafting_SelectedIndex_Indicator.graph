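# Translucent UI material that, by its name, draws the arrow indicator for the selected index in the inventory crafting HUD.
# The MF_UI_Rotator_Fractions function turns the UVs by a Rotation parameter, defaulting to a quarter, to orient a reversed tree-arrow icon texture.
# The icon's alpha is multiplied by FinalAlpha and saturated as a mask, while a FinalColor parameter, yellow by default, supplies a flat tint.
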Material /Game/Athena/HUD/Inventory/Art/M_UI_HUD_Inventory_Crafting_SelectedIndex_Indicator
graph {
 properties {blendMode=BLEND_Translucent, materialDomain=MD_UI, pixelDepthOffsetMode=PDOM_Legacy, usedShadingModels=Default Lit}
 nodes [
  n1 = MaterialFunctionCall!(Rotation_Fraction=ScalarParameter() {defaultValue=0.25, parameterName=Rotation}, TexCoord=TextureCoordinate()) {materialFunction=/FortUILibrary/Foundation/Art_ToolKit/AngledBoxes/MF_UI_Rotator_Fractions}
  Saturate(None=Multiply(
    A=TextureSample(UVs=@n1.Result) {texture=/Game/UI/SocialPanel/Art/Textures/T_UI_Icon_TreeArrow_64_reversed}.A
    B=ScalarParameter() {defaultValue=1, parameterName=FinalAlpha}
   ))
  VectorParameter() {parameterName=FinalColor, defaultValue=<rgba 1 1 0 1>}
 ]
}
