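# Quick bar slot button widget for the Athena crafting UI, showing a slotted item or an empty slot.
# Slotted-item changes resize its width by slots taken and toggle icon and empty image; non-craftable items show a disabled overlay.
# Selection plays arrow, pulse and, for valid items, underline and Equip prompt.
# Hover effects follow selection on gamepad and pointer hover otherwise; owner deactivation reverses all animations to their start.
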WidgetBlueprint /Crafting/UI/AthenaCraftingQuickBarButton
animations [
 EmptySlotHover {
  binding_count=1
  length_seconds=0.1
  object=&self:EmptySlotHover
  sequence {bindings [EmptyImage {
     id=#1
     possessed_class=Image
     tracks [{
       type=MovieSceneWidgetMaterialTrack
       sections [{type=MovieSceneParameterSection, channels [Animate_Alpha {type=MovieSceneScriptingFloatChannel, keys [[0,0],[2,1]]}]}]
      }]
    }]}
 }
 TranslateIcon {binding_count=0, length_seconds=0.1, object=&self:TranslateIcon, sequence {bindings []}}
 Scale {
  binding_count=1
  length_seconds=0.1
  object=&self:Scale
  sequence {bindings [OverlayContent {
     id=#2
     possessed_class=Overlay
     tracks [{
       type=MovieScene2DTransformTrack
       sections [{
         type=MovieScene2DTransformSection
         channels [
          Translation.X {type=MovieSceneScriptingFloatChannel}
          Translation.Y {type=MovieSceneScriptingFloatChannel}
          Angle {type=MovieSceneScriptingFloatChannel}
          Scale.X {type=MovieSceneScriptingFloatChannel, keys [[0,1],[2,1.05]]}
          Scale.Y {type=MovieSceneScriptingFloatChannel, keys [[0,1],[2,1.05]]}
          Shear.X {type=MovieSceneScriptingFloatChannel}
          Shear.Y {type=MovieSceneScriptingFloatChannel}
         ]
        }]
      }]
    }]}
 }
 ShowCraftingIndicatorArrow {
  binding_count=1
  length_seconds=0.25
  object=&self:ShowCraftingIndicatorArrow
  sequence {bindings [CraftingPanelSelectedIndicatorArrow {
     id=#3
     possessed_class=Image
     tracks [
      {
       type=MovieSceneFloatTrack
       display_name=Render Opacity
       sections [{type=MovieSceneFloatSection, channels [None {type=MovieSceneScriptingFloatChannel, keys [[0,0],[2,1]]}]}]
      }
      {
       type=MovieScene2DTransformTrack
       sections [{
         type=MovieScene2DTransformSection
         channels [
          Translation.X {type=MovieSceneScriptingFloatChannel, const [0,1]}
          Translation.Y {type=MovieSceneScriptingFloatChannel, keys [[0,-128],[2,-72],[5,-80]]}
          Scale.X {type=MovieSceneScriptingFloatChannel, const [0.9,1]}
          Scale.Y {type=MovieSceneScriptingFloatChannel, const [0.9,1]}
         ]
        }]
      }
     ]
    }]}
 }
 UnderlineSelect {
  binding_count=1
  length_seconds=0.2
  object=&self:UnderlineSelect
  sequence {bindings [Underline {
     id=#4
     possessed_class=Image
     tracks [{
       type=MovieScene2DTransformTrack
       sections [{
         type=MovieScene2DTransformSection
         channels [
          Translation.X {type=MovieSceneScriptingFloatChannel}
          Translation.Y {type=MovieSceneScriptingFloatChannel}
          Angle {type=MovieSceneScriptingFloatChannel}
          Scale.X {type=MovieSceneScriptingFloatChannel, keys [[0,0],[4,1]]}
          Scale.Y {type=MovieSceneScriptingFloatChannel}
          Shear.X {type=MovieSceneScriptingFloatChannel}
          Shear.Y {type=MovieSceneScriptingFloatChannel}
         ]
        }]
      }]
    }]}
 }
 PulseX {
  binding_count=2
  length_seconds=0.5
  object=&self:PulseX
  sequence {
   bindings [
    Image_108 {
     id=#5
     possessed_class=Image
     tracks [{
       type=MovieSceneWidgetMaterialTrack
       sections [{type=MovieSceneParameterSection, channels [Animate_Color {type=MovieSceneScriptingFloatChannel, keys [[0,0],[8,0.1],[30,0]]}]}]
      }]
    }
    Image_218 {
     id=#6
     possessed_class=Image
     tracks [
      {
       type=MovieScene2DTransformTrack
       sections [{
         type=MovieScene2DTransformSection
         channels [
          Scale.X {type=MovieSceneScriptingFloatChannel, keys [[0,1],[3,0.95],[6,1.1],[14,1]]}
          Scale.Y {type=MovieSceneScriptingFloatChannel, keys [[0,1],[3,0.95],[6,1.1],[14,1]]}
         ]
        }]
      }
      {
       type=MovieSceneWidgetMaterialTrack
       sections [{
         type=MovieSceneParameterSection
         end_frame=30
         start_frame=0
         channels [
          Scale_Texture_U {type=MovieSceneScriptingFloatChannel, keys [[3,0.6],[5,0.55],[8,0.7],[17,0.6]]}
          Scale_Texture_V {type=MovieSceneScriptingFloatChannel, keys [[3,0.6],[5,0.55],[8,0.7],[17,0.6]]}
         ]
        }]
      }
     ]
    }
   ]
   display_rate {numerator=60}
  }
 }
]
graphs [
 EventGraph {nodes [
   n1 = Event OnIsCraftableItemChanged!() -> (OutputDelegate, then, bIsCraftableItem:bool) {is=title K2Node_Event}
   > SetVisibility(
    self=Get DisabledOverlay() -> (DisabledOverlay:Overlay Object Reference)
    InVisibility=Select(Option 0=SelfHitTestInvisible, Option 1=Collapsed, Index=@n1.bIsCraftableItem)
   )
   Event On Hovered()
   n2 = Get CommonInputSubsystem()
   > n3 = Branch(Condition="Equal (Enum)"(A=GetCurrentInputType(self=@n2), B=Gamepad))
   ShowHoveredState(execute=@n3.else)
   n4 = Event OnSlottedItemChanged!() -> (OutputDelegate, then, SlottedItem:Fort Item Object Reference) {is=title K2Node_Event}
   > UpdateWidthHeight!(execute, self, Item:Fort Item Object Reference=@n4.SlottedItem) -> (then)
   n5 = IsValid(Object=@n4.SlottedItem)
   > SetVisibility(
    self=Get ItemIcon() -> (ItemIcon:Athena Item Icon Object Reference)
    InVisibility=Select(Option 0=Collapsed, Option 1=SelfHitTestInvisible, Index=@n5)
   )
   > SetVisibility(
    self=Get EmptyImage() -> (EmptyImage:Image Object Reference)
    InVisibility=Select(Option 0=SelfHitTestInvisible, Option 1=Collapsed, Index=@n5)
   )
   > Set bHasValidItem(bHasValidItem:bool=@n5) -> (Output_Get:bool)
   Event On Unhovered()
   > ShowUnhoveredState()
   Event On Deselected()
   n6 = Get ShowCraftingIndicatorArrow() -> (ShowCraftingIndicatorArrow:Widget Animation Object Reference)
   > PlayAnimation(InAnimation=@n6, StartAtTime=GetEndTime(self=@n6), PlayMode=Reverse)
   > n7 = Sequence()
   > n8 = Cast To AthenaItemIconLoadout!(execute, Object=Get ItemIcon()) -> (then, CastFailed, AsAthena Item Icon Loadout:Athena Item Icon Loadout Object Reference) {is=title K2Node_DynamicCast}
   > SetSelectedState!(execute, self:Athena Item Icon Loadout Object Reference=@n8, bIsSelected:bool=false) -> (then)
   n9 = Get UnderlineSelect() -> (UnderlineSelect:Widget Animation Object Reference)
   > PlayAnimation(InAnimation=@n9, StartAtTime=GetEndTime(self=@n9), PlayMode=Reverse)
   n10 = Get CommonInputSubsystem()
   Branch(execute=@n7.then_1, Condition="Equal (Enum)"(A=GetCurrentInputType(self=@n10), B=Gamepad))
   > ShowUnhoveredState()
   Event On Selected()
   > ShowSelectedState()
   n11 = Get CommonInputSubsystem()
   > Branch(Condition="Equal (Enum)"(A=GetCurrentInputType(self=@n11), B=Gamepad))
   > ShowHoveredState()
   Event OnOwnerDeactivated() {is=title K2Node_Event}
   > SetVisibility(self=Get ScaleBox_InputAction() -> (ScaleBox_InputAction:Scale Box Object Reference), InVisibility=Collapsed)
   > ResetAnims()
   > n12 = Cast To AthenaItemIconLoadout(Object=Get ItemIcon())
   > ResetAnims(self:Athena Item Icon Loadout Object Reference=@n12)
  ]}
 ResetAnimToStart {nodes [
   n1 = Inputs!() -> (Exec:Exec, WidgetAnim:Widget Animation Object Reference)
   > PlayAnimation(InAnimation=@n1, StartAtTime=GetEndTime(self=@n1), PlayMode=Reverse)
   Outputs()
  ]}
 ResetAnims {nodes [
   ResetAnims() {is=title K2Node_FunctionEntry}
   > n1 = For Each Loop(Array:Array of Widget Animation Object References=Make Array!(
     "[0]":Widget Animation Object Reference=Get UnderlineSelect()
     "[1]":Widget Animation Object Reference=Get ShowCraftingIndicatorArrow()
     "[2]":Widget Animation Object Reference=Get Scale() -> (Scale:Widget Animation Object Reference)
     "[3]":Widget Animation Object Reference=Get EmptySlotHover() -> (EmptySlotHover:Widget Animation Object Reference)
     "[4]":Widget Animation Object Reference=Get PulseX() -> (PulseX:Widget Animation Object Reference)
     "[5]":Widget Animation Object Reference=Get TranslateIcon() -> (TranslateIcon:Widget Animation Object Reference)
    ) -> (Array:Array of Widget Animation Object References)) -> (Array Element:Widget Animation Object Reference)
   > Reset Anim to Start!(Exec:Exec, WidgetAnim:Widget Animation Object Reference=@n1) {is=title K2Node_MacroInstance}
  ]}
 ShowHoveredState {nodes [
   ShowHoveredState() {is=title K2Node_FunctionEntry}
   > n1 = Branch(Condition=Get bHasValidItem() -> (bHasValidItem:bool))
   > n2 = Cast To AthenaItemIconLoadout(Object=Get ItemIcon())
   > SetHoveredState!(execute, self:Athena Item Icon Loadout Object Reference=@n2, bIsHovered:bool=true) -> (then)
   > PlayAnimationForward(InAnimation=Get Scale())
   > PlayAnimationForward(InAnimation=Get TranslateIcon())
   PlayAnimationForward(execute=@n1.else, InAnimation=Get EmptySlotHover())
  ]}
 ShowSelectedState {nodes [
   ShowSelectedState() {is=title K2Node_FunctionEntry}
   > PlayAnimationForward(InAnimation=Get ShowCraftingIndicatorArrow())
   > PlayAnimationForward(InAnimation=Get PulseX())
   n1 = Get bHasValidItem()
   > SetVisibility(self=Get ScaleBox_InputAction(), InVisibility=Select(Option 0=Collapsed, Option 1=SelfHitTestInvisible, Index=@n1))
   > Branch(Condition=@n1)
   > n2 = Cast To AthenaItemIconLoadout(Object=Get ItemIcon())
   > SetSelectedState(self=@n2, bIsSelected=true)
   > PlayAnimation(InAnimation=Get UnderlineSelect())
  ]}
 ShowUnhoveredState {nodes [
   ShowUnhoveredState() {is=title K2Node_FunctionEntry}
   > n1 = Branch(Condition=Get bHasValidItem())
   > n2 = Cast To AthenaItemIconLoadout(Object=Get ItemIcon())
   > SetHoveredState(self=@n2, bIsHovered=false)
   > PlayAnimationReverse(InAnimation=Get Scale())
   > PlayAnimationReverse(InAnimation=Get TranslateIcon())
   PlayAnimationReverse(execute=@n1.else, InAnimation=Get EmptySlotHover())
  ]}
 UpdateWidthHeight {nodes [
   n1 = UpdateWidthHeight!() -> (then, Item:Fort Item Object Reference) {is=title K2Node_FunctionEntry}
   > n2 = Is Valid(InputObject=@n1)
   > n3 = Set Local_NumSlotsTaken(Local_NumSlotsTaken:int="To Integer (Byte)"!(InByte:Byte=GetNumberOfSlotsTaken!(self:Fort Item Object Reference=@n1) -> (ReturnValue:Byte=0)) -> (ReturnValue:int=0)) -> (Output_Get:int=0)
   n4 = "Max (Integer)"(A=Get Local_NumSlotsTaken() -> (Local_NumSlotsTaken:int=0), B=1)
   SetWidthOverride(
    execute=[@n3.then, @n2.Is Not Valid]
    self=Get SlotSizeBox() -> (SlotSizeBox:Size Box Object Reference)
    InWidthOverride=float * float(A=Get SingleSlotSize() -> (SingleSlotSize:double), B="To Float (Integer)"(InInt=@n4))
   )
  ]}
]
variables [
 bHasValidItem {type=boolean}
 SingleSlotSize {type=number, default=122}
]
widgets {
 info {inheritedWidgetCount=2, parentClass=/Script/CraftingUI.AthenaCraftingQuickBarButton, rootWidgetClass=&SizeBox, widgetCount=12}
 widgets [
  SlotSizeBox {
   widgetClassPath=&SizeBox
   bIsVariable=true
   widget=SlotSizeBox
   properties {bOverride_HeightOverride=true, bOverride_WidthOverride=true, heightOverride=122, widthOverride=122}
  }
  Overlay_179 {widgetClassPath=&Overlay, parent=SlotSizeBox, widget=Overlay_179, slot {class=SizeBoxSlot}}
  OverlayContent {
   widgetClassPath=&Overlay
   parent=Overlay_179
   widget=OverlayContent
   slot {class=OverlaySlot, properties {horizontalAlignment=HAlign_Fill, verticalAlignment=VAlign_Fill}}
  }
  EmptyImage {
   widgetClassPath=&Image
   bIsVariable=true
   parent=OverlayContent
   widget=EmptyImage
   properties {
    visibility=SelfHitTestInvisible
    brush {
     bIsDynamicallyLoaded=false
     drawAs=Box
     imageType=NoImage
     mirroring=NoMirror
     resourceName=None
     resourceObject=/Game/Athena/HUD/Inventory/Art/M_UI_Inventory_LoadoutItem_EmptySlot
     tiling=NoTile
     imageSize=<64 64>
     margin=<ltrb 0.5 0.5 0.5 0.5>
     outlineSettings {
      bUseBrushTransparency=false
      roundingType=HalfHeightRadius
      width=0
      color {colorUseRule=UseColor_Specified, specifiedColor=<rgba 0 0 0 0>}
      cornerRadii=<0 0 0 0>
     }
     tintColor {colorUseRule=UseColor_Specified, specifiedColor=<rgba 1 1 1 1>}
     uVRegion {bIsValid=false, max=<0 0>, min=<0 0>}
    }
   }
   slot {class=OverlaySlot, properties {horizontalAlignment=HAlign_Fill, verticalAlignment=VAlign_Fill}}
  }
  ItemIcon {
   widgetClassPath=/Game/Athena/HUD/Inventory/AthenaItemIconLoadout
   bInherited=true
   parent=OverlayContent
   widget=ItemIcon
   slot {class=OverlaySlot, properties {horizontalAlignment=HAlign_Fill, verticalAlignment=VAlign_Fill}}
  }
  DisabledOverlay {
   widgetClassPath=&Overlay
   bIsVariable=true
   parent=OverlayContent
   widget=DisabledOverlay
   slot {class=OverlaySlot, properties {horizontalAlignment=HAlign_Fill, verticalAlignment=VAlign_Fill}}
  }
  Image_108 {
   widgetClassPath=&Image
   parent=DisabledOverlay
   widget=Image_108
   properties {
    visibility=SelfHitTestInvisible
    brush {
     bIsDynamicallyLoaded=false
     drawAs=Image
     imageType=NoImage
     mirroring=NoMirror
     resourceName=None
     resourceObject=/Game/Athena/HUD/Inventory/Art/M_UI_HUD_CraftingDisabled_EquipSlot_Backing
     tiling=NoTile
     imageSize=<32 32>
     margin=<ltrb 0 0 0 0>
     outlineSettings {
      bUseBrushTransparency=false
      roundingType=HalfHeightRadius
      width=0
      color {colorUseRule=UseColor_Specified, specifiedColor=<rgba 0 0 0 0>}
      cornerRadii=<0 0 0 0>
     }
     tintColor {colorUseRule=UseColor_Specified, specifiedColor=<rgba 1 1 1 1>}
     uVRegion {bIsValid=false, max=<0 0>, min=<0 0>}
    }
   }
   slot {class=OverlaySlot, properties {horizontalAlignment=HAlign_Fill, verticalAlignment=VAlign_Fill}}
  }
  Image_218 {
   widgetClassPath=&Image
   parent=DisabledOverlay
   widget=Image_218
   properties {
    visibility=SelfHitTestInvisible
    brush {
     bIsDynamicallyLoaded=false
     drawAs=Image
     imageType=NoImage
     mirroring=NoMirror
     resourceName=None
     resourceObject=/FortUILibrary/Foundation/Art_ToolKit/AngledBoxes/M_UI_CraftingDisabledIndicator
     tiling=NoTile
     imageSize=<44 44>
     margin=<ltrb 0 0 0 0>
     outlineSettings {
      bUseBrushTransparency=false
      roundingType=HalfHeightRadius
      width=0
      color {colorUseRule=UseColor_Specified, specifiedColor=<rgba 0 0 0 0>}
      cornerRadii=<0 0 0 0>
     }
     tintColor {colorUseRule=UseColor_Specified, specifiedColor=<rgba 1 1 1 1>}
     uVRegion {bIsValid=false, max=<0 0>, min=<0 0>}
    }
   }
   slot {class=OverlaySlot, properties {horizontalAlignment=HAlign_Center, verticalAlignment=VAlign_Center}}
  }
  Underline {
   widgetClassPath=&Image
   parent=Overlay_179
   widget=Underline
   properties {
    visibility=SelfHitTestInvisible
    brush {
     bIsDynamicallyLoaded=false
     drawAs=Image
     imageType=NoImage
     mirroring=NoMirror
     resourceName=None
     resourceObject=null
     tiling=NoTile
     imageSize=<32 2>
     margin=<ltrb 0 0 0 0>
     outlineSettings {
      bUseBrushTransparency=false
      roundingType=HalfHeightRadius
      width=0
      color {colorUseRule=UseColor_Specified, specifiedColor=<rgba 0 0 0 0>}
      cornerRadii=<0 0 0 0>
     }
     tintColor {colorUseRule=UseColor_Specified, specifiedColor=<rgba 1 1 1 1>}
     uVRegion {bIsValid=false, max=<0 0>, min=<0 0>}
    }
    renderTransform {angle=0, scale=<0 1>, shear=<0 0>, translation=<0 18>}
   }
   slot {class=OverlaySlot, properties {horizontalAlignment=HAlign_Fill, verticalAlignment=VAlign_Bottom}}
  }
  ScaleBox_InputAction {
   widgetClassPath=&ScaleBox
   bIsVariable=true
   parent=Overlay_179
   widget=ScaleBox_InputAction
   properties {stretch=UserSpecified, userSpecifiedScale=0.9, renderTransform {angle=0, scale=<1 1>, shear=<0 0>, translation=<0 54>}}
   slot {class=OverlaySlot, properties {horizontalAlignment=HAlign_Center, verticalAlignment=VAlign_Bottom}}
  }
  InputActionWidget {
   widgetClassPath=&CommonActionWidget
   bInherited=true
   parent=ScaleBox_InputAction
   widget=InputActionWidget
   properties {
    visibility=SelfHitTestInvisible
    iconRimBrush {
     bIsDynamicallyLoaded=false
     drawAs=Image
     imageType=NoImage
     mirroring=NoMirror
     resourceName=None
     resourceObject=/Game/Athena/HUD/Inventory/Art/M_UI_HUD_Inventory_SelectionPrompt_Pulse
     tiling=NoTile
     imageSize=<80 80>
     margin=<ltrb 0 0 0 0>
     outlineSettings {
      bUseBrushTransparency=false
      roundingType=HalfHeightRadius
      width=0
      color {colorUseRule=UseColor_Specified, specifiedColor=<rgba 0 0 0 0>}
      cornerRadii=<0 0 0 0>
     }
     tintColor {colorUseRule=UseColor_Specified, specifiedColor=<rgba 1 1 1 1>}
     uVRegion {bIsValid=false, max=<0 0>, min=<0 0>}
    }
    inputActions [{dataTable=&InputActionDataTable, rowName=Equip}]
   }
   slot {class=ScaleBoxSlot}
  }
  CraftingPanelSelectedIndicatorArrow {
   widgetClassPath=&Image
   parent=Overlay_179
   widget=CraftingPanelSelectedIndicatorArrow
   properties {
    clipping=ClipToBoundsWithoutIntersecting
    renderOpacity=0
    visibility=SelfHitTestInvisible
    brush {
     bIsDynamicallyLoaded=false
     drawAs=Image
     imageType=NoImage
     mirroring=NoMirror
     resourceName=None
     resourceObject=/Game/Athena/HUD/Inventory/Art/M_UI_HUD_Inventory_Crafting_SelectedIndex_Indicator
     tiling=NoTile
     imageSize=<40 40>
     margin=<ltrb 0 0 0 0>
     outlineSettings {
      bUseBrushTransparency=false
      roundingType=HalfHeightRadius
      width=0
      color {colorUseRule=UseColor_Specified, specifiedColor=<rgba 0 0 0 0>}
      cornerRadii=<0 0 0 0>
     }
     tintColor {colorUseRule=UseColor_Specified, specifiedColor=<rgba 1 1 1 1>}
     uVRegion {bIsValid=false, max=<0 0>, min=<0 0>}
    }
    renderTransform {angle=0, scale=<1 1>, shear=<0 0>, translation=<0 -80>}
   }
   slot {class=OverlaySlot, properties {horizontalAlignment=HAlign_Center}}
  }
 ]
}
