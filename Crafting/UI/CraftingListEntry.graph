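# Crafting list entry widget showing an item icon card with hover, selection and can't-craft states.
# Button and list events drive hover, selection and opacity animations, toggling the icon's recipe hover state.
# When an item is set, it copies the icon material's hover color onto the selected backplate and shows a disabled overlay if uncraftable.
# Mobile hides the side selector and enlarges the card; a helper shows scroll hints for gamepad input.
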WidgetBlueprint /Crafting/UI/CraftingListEntry
animations [
 Hover {
  binding_count=2
  length_seconds=0.1
  object=&self:Hover
  sequence {bindings [
    Shadow {
     id=#1
     possessed_class=Image
     tracks [
      {
       type=MovieSceneWidgetMaterialTrack
       sections [{type=MovieSceneParameterSection, channels [Animate_Alpha {type=MovieSceneScriptingFloatChannel, keys [[0,0],[2,1]]}]}]
      }
      {
       type=MovieScene2DTransformTrack
       sections [{
         type=MovieScene2DTransformSection
         channels [
          Translation.Y {type=MovieSceneScriptingFloatChannel, keys [[0,0],[2,16]]}
          Scale.X {type=MovieSceneScriptingFloatChannel, keys [[0,1.25],[2,1.5]]}
          Scale.Y {type=MovieSceneScriptingFloatChannel, keys [[0,1.25],[2,1.5]]}
         ]
        }]
      }
     ]
    }
    Card {
     id=#2
     possessed_class=Overlay
     tracks [{
       type=MovieScene2DTransformTrack
       sections [{
         type=MovieScene2DTransformSection
         channels [
          Translation.X {type=MovieSceneScriptingFloatChannel}
          Translation.Y {type=MovieSceneScriptingFloatChannel, keys [[0,0],[2,-2]]}
          Scale.X {type=MovieSceneScriptingFloatChannel, keys [[0,1],[2,1.075]]}
          Scale.Y {type=MovieSceneScriptingFloatChannel, keys [[0,1],[2,1.075]]}
         ]
        }]
      }]
    }
   ]}
 }
 Selected {
  binding_count=2
  length_seconds=0.3
  object=&self:Selected
  sequence {bindings [
    BackplateSelected {
     id=#3
     possessed_class=Image
     tracks [
      {
       type=MovieSceneWidgetMaterialTrack
       sections [{type=MovieSceneParameterSection, channels [Animate_Alpha {type=MovieSceneScriptingFloatChannel, keys [[0,0],[4,1]]}]}]
      }
      {
       type=MovieScene2DTransformTrack
       sections [{
         type=MovieScene2DTransformSection
         channels [
          Scale.X {type=MovieSceneScriptingFloatChannel, const [1,1]}
          Scale.Y {type=MovieSceneScriptingFloatChannel, const [1,1]}
         ]
        }]
      }
      {
       type=MovieSceneByteTrack
       display_name=Visibility
       sections [{type=MovieSceneByteSection, end_frame=4, start_frame=0, channels [None {type=MovieSceneScriptingByteChannel, keys [[0,1],[1,4]]}]}]
      }
     ]
    }
    SideSelector {
     id=#4
     possessed_class=Image
     tracks [
      {
       type=MovieSceneWidgetMaterialTrack
       sections [{
         type=MovieSceneParameterSection
         channels [
          Offset_Angle_Top {type=MovieSceneScriptingFloatChannel, keys [[0,-1],[1.9997,-0.15],[6,-0.35]]}
          Plate_Angle {type=MovieSceneScriptingFloatChannel, keys [[0,0],[1.9997,0.04]]}
         ]
        }]
      }
      {
       type=MovieSceneByteTrack
       display_name=Visibility
       sections [{type=MovieSceneByteSection, end_frame=5, start_frame=0, channels [None {type=MovieSceneScriptingByteChannel, keys [[0,1],[1,4]]}]}]
      }
     ]
    }
   ]}
 }
 AnimateCantCraftAlpha {
  binding_count=1
  length_seconds=0.1
  object=&self:AnimateCantCraftAlpha
  sequence {bindings [ItemIconBorder {
     id=#5
     possessed_class=Border
     tracks [{
       type=MovieSceneFloatTrack
       display_name=Render Opacity
       sections [{type=MovieSceneFloatSection, channels [None {type=MovieSceneScriptingFloatChannel, keys [[0,0.5],[2,1]]}]}]
      }]
    }]}
 }
]
graphs [
 EventGraph {nodes [
   n1 = On Hovered State Changed!() -> (OutputDelegate, then, bIsHovered:bool=~) {is=title K2Node_CustomEvent}
   > n2 = Branch(Condition=OR Boolean(A=@n1.bIsHovered, B=GetSelected!(self:Common Button Base Object Reference) -> (ReturnValue:bool)))
   > n3 = Branch(Condition=Get IsSelected() -> (IsSelected:bool))
   n4 = Cast To AthenaItemIconLoadout!(execute=@n3.else, Object=Get ItemIcon() -> (ItemIcon:Athena Item Icon Object Reference)) -> (then, CastFailed, AsAthena Item Icon Loadout:Athena Item Icon Loadout Object Reference) {is=title K2Node_DynamicCast}
   > RecipeHoverState!(execute, self:Athena Item Icon Loadout Object Reference=@n4, IsHovered:bool=true) -> (then)
   > PlayAnimationForward(InAnimation=Get Hover() -> (Hover:Widget Animation Object Reference))
   n5 = Branch(execute=@n2.else, Condition=Get IsSelected())
   n6 = Cast To AthenaItemIconLoadout(execute=@n5.else, Object=Get ItemIcon())
   > RecipeHoverState(self=@n6, IsHovered=false)
   > PlayAnimationReverse(InAnimation=Get Hover())
   n7 = On Selected Changed!() -> (OutputDelegate, then, bIsSelected:bool=~) {is=title K2Node_CustomEvent}
   > On Hovered State Changed!(execute, self, bIsHovered:bool=IsHovered()) -> (then) {is=title K2Node_CallFunction}
   > n8 = Branch(Condition=@n7.bIsSelected)
   > Set IsSelected(IsSelected:bool=true) -> (Output_Get:bool)
   > PlayAnimationForward(InAnimation=Get Selected() -> (Selected:Widget Animation Object Reference))
   > n9 = Cast To AthenaItemIconLoadout(Object=Get ItemIcon())
   > ResetRecipeHoverState(self:Athena Item Icon Loadout Object Reference=@n9)
   > n10 = Branch(Condition=Get bCanCraftItem() -> (bCanCraftItem:bool))
   PlayAnimationForward(execute=@n10.else, InAnimation=Get AnimateCantCraftAlpha() -> (AnimateCantCraftAlpha:Widget Animation Object Reference))
   Set IsSelected(execute=@n8.else, IsSelected=false)
   n11 = Get Selected()
   > PlayAnimation(InAnimation=@n11, StartAtTime=GetEndTime(self=@n11), PlayMode=Reverse)
   > n12 = Cast To AthenaItemIconLoadout(Object=Get ItemIcon())
   > RecipeHoverState(self=@n12, IsHovered=false)
   > PlayAnimationReverse(InAnimation=Get Hover())
   > n13 = Branch(Condition=Get bCanCraftItem())
   PlayAnimationReverse(execute=@n13.else, InAnimation=Get AnimateCantCraftAlpha(), PlaybackSpeed=4)
   Event On Entry Released() {is=title K2Node_Event}
   > SetVisibility(InVisibility=Collapsed)
   Event Construct()
   > n14 = SetVisibility(InVisibility=Collapsed)
   n15 = Cast To AthenaItemIconLoadout(execute=[@n14.then, @n27.then], Object=Get ItemIcon())
   > n16 = GetDynamicMaterial(self=Get Image_ItemIcon!(self:Athena Item Icon Object Reference=@n15) -> (Image_ItemIcon:Image Object Reference))
   > SetScalarParameterValue(self=@n16, ParameterName=IsRecipeView, Value=1)
   > ResetAnims()
   > SetVisibility(
    self=Get SideSelectorBorder() -> (SideSelectorBorder:Border Object Reference)
    InVisibility=Select(Option 0=SelfHitTestInvisible, Option 1=Collapsed, Index=IsMobileGame())
   )
   > Branch(Condition=IsMobileGame())
   n17 = Get SizeBox_0() -> (SizeBox_0:Size Box Object Reference)
   > SetWidthOverride(self=@n17, InWidthOverride=110)
   > SetHeightOverride(self=@n17, InHeightOverride=110)
   Event On Selected()
   > On Selected Changed!(execute, self, bIsSelected:bool=true) -> (then) {is=title K2Node_CallFunction}
   Event OnCraftingListItemSet() {is=title K2Node_Event}
   > SetVisibility(InVisibility=Visible)
   > On Selected Changed(bIsSelected=GetSelected())
   > On Hovered State Changed(bIsHovered=IsHovered())
   > n18 = Branch(Condition=Get bCanCraftItem())
   n19 = Get AnimateCantCraftAlpha()
   n20 = GetEndTime(self=@n19)
   > n21 = PlayAnimation(InAnimation=@n19, StartAtTime=@n20)
   n22 = Cast To AthenaItemIconLoadout(execute=[@n26.then, @n21.then], Object=Get ItemIcon())
   > n23 = GetDynamicMaterial(self=Get Image_ItemIcon(self=@n22))
   > n24 = Get Vector Parameter Value!(execute, self:Material Instance Dynamic Object Reference=@n23, ParameterName:Name=BG_Hotspot_Color_Hover) -> (then, ReturnValue:Linear Color Structure) {is=title K2Node_CallFunction}
   > n25 = GetDynamicMaterial(self=Get BackplateSelected() -> (BackplateSelected:Image Object Reference))
   > SetVectorParameterValue(self=@n25, ParameterName=Color_Start, Value=@n24)
   > SetVisibility(
    self=Get CraftingDisabledStateOverlay() -> (CraftingDisabledStateOverlay:Overlay Object Reference)
    InVisibility=Select(Option 0=SelfHitTestInvisible, Option 1=Collapsed, Index=Get bCanCraftItem())
   )
   n26 = PlayAnimation(execute=@n18.else, InAnimation=@n19, StartAtTime=@n20, PlayMode=Reverse)
   Event On Deselected()
   > On Selected Changed(bIsSelected=false)
   Event On Hovered()
   > On Hovered State Changed(bIsHovered=true)
   Event On Unhovered()
   > On Hovered State Changed(bIsHovered=false)
   n27 = Event PreConstruct()
  ]}
 ResetAnims {nodes [
   ResetAnims() {is=title K2Node_FunctionEntry}
   n1 = Get AnimateCantCraftAlpha()
   > PlayAnimation(InAnimation=@n1, StartAtTime=GetEndTime(self=@n1), PlayMode=Reverse)
   n2 = Get Selected()
   > PlayAnimation(InAnimation=@n2, StartAtTime=GetEndTime(self=@n2), PlayMode=Reverse)
   n3 = Get Hover()
   > PlayAnimation(InAnimation=@n3, StartAtTime=GetEndTime(self=@n3), PlayMode=Reverse)
  ]}
 SetNavIndicatorForGamepad {nodes [
   SetNavIndicatorForGamepad() {is=title K2Node_FunctionEntry}
   n1 = GetCurrentInputType(self=Get CommonInputSubsystem!() -> (ReturnValue:Common Input Object Reference) {is=title K2Node_GetBlueprintContextEx})
   > SetVisibility(
    self=Get ControllerNavIndicator() -> (ControllerNavIndicator:Vertical Box Object Reference)
    InVisibility=Select(Option 0=Collapsed, Option 1=SelfHitTestInvisible, Index="Equal (Enum)"(A=@n1, B=Gamepad))
   )
  ]}
]
variables [IsSelected {type=boolean}]
widgets {
 info {inheritedWidgetCount=2, parentClass=/Script/CraftingUI.FortCraftingListEntry, rootWidgetClass=&Overlay, widgetCount=17}
 widgets [
  Overlay_0 {widgetClassPath=&Overlay, widget=Overlay_0}
  SizeBox_0 {
   widgetClassPath=&SizeBox
   bIsVariable=true
   parent=Overlay_0
   widget=SizeBox_0
   properties {bOverride_HeightOverride=true, bOverride_WidthOverride=true, heightOverride=88, widthOverride=88}
   slot {class=OverlaySlot}
  }
  Card {widgetClassPath=&Overlay, parent=SizeBox_0, widget=Card, slot {class=SizeBoxSlot}}
  Shadow {
   widgetClassPath=&Image
   bIsVariable=true
   parent=Card
   widget=Shadow
   properties {
    clipping=ClipToBoundsWithoutIntersecting
    visibility=SelfHitTestInvisible
    brush {
     bIsDynamicallyLoaded=false
     drawAs=Box
     imageType=NoImage
     mirroring=NoMirror
     resourceName=None
     resourceObject=/Game/Athena/HUD/Inventory/Art/M_UI_Inventory_Loadout_DropShadow
     tiling=NoTile
     imageSize=<80 80>
     margin=<ltrb 0.5 0.5 0.5 0.5>
     outlineSettings {
      bUseBrushTransparency=false
      roundingType=HalfHeightRadius
      width=0
      color {colorUseRule=UseColor_Specified, specifiedColor=<rgba 0 0 0 0>}
      cornerRadii=<0 0 0 0>
     }
     tintColor {colorUseRule=UseColor_Specified, specifiedColor=<rgba 1 1 1 1>}
     uVRegion {bIsValid=false, max=<0 0>, min=<0 0>}
    }
    renderTransform {angle=0, scale=<1.5 1.5>, shear=<0 0>, translation=<0 16>}
   }
   slot {class=OverlaySlot, properties {horizontalAlignment=HAlign_Fill, verticalAlignment=VAlign_Fill}}
  }
  BackplateSelected {
   widgetClassPath=&Image
   bIsVariable=true
   parent=Card
   widget=BackplateSelected
   properties {
    clipping=ClipToBoundsWithoutIntersecting
    visibility=Collapsed
    brush {
     bIsDynamicallyLoaded=false
     drawAs=Box
     imageType=NoImage
     mirroring=NoMirror
     resourceName=None
     resourceObject=/Game/Athena/HUD/Inventory/Art/M_UI_HUD_CraftingEntry_RarityBorder
     tiling=NoTile
     imageSize=<32 32>
     margin=<ltrb 0 0 0 0>
     outlineSettings {
      bUseBrushTransparency=false
      roundingType=HalfHeightRadius
      width=0
      color {colorUseRule=UseColor_Specified, specifiedColor=<rgba 0 0 0 0>}
      cornerRadii=<0 0 0 0>
     }
     tintColor {colorUseRule=UseColor_Specified, specifiedColor=<rgba 1 1 1 1>}
     uVRegion {bIsValid=false, max=<0 0>, min=<0 0>}
    }
   }
   slot {class=OverlaySlot, properties {horizontalAlignment=HAlign_Fill, verticalAlignment=VAlign_Fill}}
  }
  ItemIconBorder {
   widgetClassPath=&Border
   bIsVariable=true
   parent=Card
   widget=ItemIconBorder
   properties {
    visibility=SelfHitTestInvisible
    background {
     bIsDynamicallyLoaded=false
     drawAs=NoDrawType
     imageType=NoImage
     mirroring=NoMirror
     resourceName=None
     resourceObject=null
     tiling=NoTile
     imageSize=<32 32>
     margin=<ltrb 0 0 0 0>
     outlineSettings {
      bUseBrushTransparency=false
      roundingType=HalfHeightRadius
      width=0
      color {colorUseRule=UseColor_Specified, specifiedColor=<rgba 0 0 0 0>}
      cornerRadii=<0 0 0 0>
     }
     tintColor {colorUseRule=UseColor_Specified, specifiedColor=<rgba 1 1 1 1>}
     uVRegion {bIsValid=false, max=<0 0>, min=<0 0>}
    }
    padding=<ltrb 0 0 0 0>
   }
   slot {class=OverlaySlot, properties {horizontalAlignment=HAlign_Center, verticalAlignment=VAlign_Center, padding=<ltrb 4 4 4 4>}}
  }
  ItemIcon {
   widgetClassPath=/Game/Athena/HUD/Inventory/AthenaItemIconLoadout
   bInherited=true
   parent=ItemIconBorder
   widget=ItemIcon
   properties {bShowAmmoIcon=false, bShowStackCountText=false}
   slot {class=BorderSlot, properties {padding=<ltrb 0 0 0 0>}}
  }
  CraftingDisabledStateOverlay {
   widgetClassPath=&Overlay
   bIsVariable=true
   parent=Card
   widget=CraftingDisabledStateOverlay
   properties {visibility=Collapsed}
   slot {class=OverlaySlot, properties {horizontalAlignment=HAlign_Fill, verticalAlignment=VAlign_Fill, padding=<ltrb 4 4 4 4>}}
  }
  DisabledStateSubLayer {
   widgetClassPath=&Overlay
   bIsVariable=true
   parent=CraftingDisabledStateOverlay
   widget=DisabledStateSubLayer
   slot {class=OverlaySlot, properties {horizontalAlignment=HAlign_Fill, verticalAlignment=VAlign_Fill}}
  }
  Image_218 {
   widgetClassPath=&Image
   bIsVariable=true
   parent=DisabledStateSubLayer
   widget=Image_218
   properties {
    visibility=SelfHitTestInvisible
    brush {
     bIsDynamicallyLoaded=false
     drawAs=Image
     imageType=NoImage
     mirroring=NoMirror
     resourceName=None
     resourceObject=/FortUILibrary/Foundation/Art_ToolKit/AngledBoxes/M_UI_CraftingDisabledIndicator
     tiling=NoTile
     imageSize=<24 24>
     margin=<ltrb 0 0 0 0>
     outlineSettings {
      bUseBrushTransparency=false
      roundingType=HalfHeightRadius
      width=0
      color {colorUseRule=UseColor_Specified, specifiedColor=<rgba 0 0 0 0>}
      cornerRadii=<0 0 0 0>
     }
     tintColor {colorUseRule=UseColor_Specified, specifiedColor=<rgba 1 1 1 1>}
     uVRegion {bIsValid=false, max=<0 0>, min=<0 0>}
    }
   }
   slot {class=OverlaySlot, properties {horizontalAlignment=HAlign_Right, verticalAlignment=VAlign_Bottom, padding=<ltrb 0 0 4 4>}}
  }
  SideSelectorBorder {
   widgetClassPath=&Border
   bIsVariable=true
   parent=Overlay_0
   widget=SideSelectorBorder
   properties {
    background {
     bIsDynamicallyLoaded=false
     drawAs=NoDrawType
     imageType=NoImage
     mirroring=NoMirror
     resourceName=None
     resourceObject=null
     tiling=NoTile
     imageSize=<32 32>
     margin=<ltrb 0 0 0 0>
     outlineSettings {
      bUseBrushTransparency=false
      roundingType=HalfHeightRadius
      width=0
      color {colorUseRule=UseColor_Specified, specifiedColor=<rgba 0 0 0 0>}
      cornerRadii=<0 0 0 0>
     }
     tintColor {colorUseRule=UseColor_Specified, specifiedColor=<rgba 1 1 1 1>}
     uVRegion {bIsValid=false, max=<0 0>, min=<0 0>}
    }
    padding=<ltrb 0 0 0 0>
   }
   slot {class=OverlaySlot, properties {verticalAlignment=VAlign_Fill}}
  }
  SideSelector {
   widgetClassPath=&Image
   bIsVariable=true
   parent=SideSelectorBorder
   widget=SideSelector
   properties {
    clipping=ClipToBoundsWithoutIntersecting
    visibility=Collapsed
    brush {
     bIsDynamicallyLoaded=false
     drawAs=Image
     imageType=NoImage
     mirroring=NoMirror
     resourceName=None
     resourceObject=/Game/Athena/HUD/Inventory/Art/M_UI_HUD_CraftingList_SideIndicator
     tiling=NoTile
     imageSize=<14 10>
     margin=<ltrb 0 0 0 0>
     outlineSettings {
      bUseBrushTransparency=false
      roundingType=HalfHeightRadius
      width=0
      color {colorUseRule=UseColor_Specified, specifiedColor=<rgba 0 0 0 0>}
      cornerRadii=<0 0 0 0>
     }
     tintColor {colorUseRule=UseColor_Specified, specifiedColor=<rgba 1 1 1 1>}
     uVRegion {bIsValid=false, max=<0 0>, min=<0 0>}
    }
    renderTransform {angle=0, scale=<1 1>, shear=<0 0>, translation=<-26 0>}
   }
   slot {class=BorderSlot, properties {padding=<ltrb 0 0 0 0>}}
  }
  ControllerNavIndicator {
   widgetClassPath=&VerticalBox
   bIsVariable=true
   parent=Overlay_0
   widget=ControllerNavIndicator
   properties {renderTransform {angle=0, scale=<1 1>, shear=<0 0>, translation=<-64 0>}}
   slot {class=OverlaySlot, properties {verticalAlignment=VAlign_Center}}
  }
  Image_171 {
   widgetClassPath=&Image
   bIsVariable=true
   parent=ControllerNavIndicator
   widget=Image_171
   properties {
    visibility=SelfHitTestInvisible
    brush {
     bIsDynamicallyLoaded=false
     drawAs=Image
     imageType=NoImage
     mirroring=NoMirror
     resourceName=None
     resourceObject=/Game/Athena/HUD/Inventory/Art/M_UI_HUD_Inventory_Crafting_List_InteractionArrow
     tiling=NoTile
     imageSize=<16 16>
     margin=<ltrb 0 0 0 0>
     outlineSettings {
      bUseBrushTransparency=false
      roundingType=HalfHeightRadius
      width=0
      color {colorUseRule=UseColor_Specified, specifiedColor=<rgba 0 0 0 0>}
      cornerRadii=<0 0 0 0>
     }
     tintColor {colorUseRule=UseColor_Specified, specifiedColor=<rgba 1 1 1 1>}
     uVRegion {bIsValid=false, max=<0 0>, min=<0 0>}
    }
    renderTransform {angle=180, scale=<1 1>, shear=<0 0>, translation=<0 0>}
   }
   slot {class=VerticalBoxSlot, properties {horizontalAlignment=HAlign_Center, verticalAlignment=VAlign_Bottom}}
  }
  ScaleBox_0 {
   widgetClassPath=&ScaleBox
   parent=ControllerNavIndicator
   widget=ScaleBox_0
   properties {stretch=UserSpecified, userSpecifiedScale=0.9}
   slot {class=VerticalBoxSlot, properties {horizontalAlignment=HAlign_Center, verticalAlignment=VAlign_Center}}
  }
  InputActionWidget {
   widgetClassPath=&CommonActionWidget
   bInherited=true
   parent=ScaleBox_0
   widget=InputActionWidget
   properties {
    visibility=SelfHitTestInvisible
    iconRimBrush {
     bIsDynamicallyLoaded=false
     drawAs=NoDrawType
     imageType=NoImage
     mirroring=NoMirror
     resourceName=None
     resourceObject=null
     tiling=NoTile
     imageSize=<40 40>
     margin=<ltrb 0 0 0 0>
     outlineSettings {
      bUseBrushTransparency=false
      roundingType=HalfHeightRadius
      width=0
      color {colorUseRule=UseColor_Specified, specifiedColor=<rgba 0 0 0 0>}
      cornerRadii=<0 0 0 0>
     }
     tintColor {colorUseRule=UseColor_Specified, specifiedColor=<rgba 1 1 1 1>}
     uVRegion {bIsValid=false, max=<0 0>, min=<0 0>}
    }
    inputActions [{dataTable=&InputActionDataTable, rowName=Scroll}]
   }
   slot {class=ScaleBoxSlot}
  }
  Image {
   widgetClassPath=&Image
   bIsVariable=true
   parent=ControllerNavIndicator
   widget=Image
   properties {
    visibility=SelfHitTestInvisible
    brush {
     bIsDynamicallyLoaded=false
     drawAs=Image
     imageType=NoImage
     mirroring=NoMirror
     resourceName=None
     resourceObject=&M_UI_HUD_Inventory_Crafting_List_InteractionArrow
     tiling=NoTile
     imageSize=<16 16>
     margin=<ltrb 0 0 0 0>
     outlineSettings {
      bUseBrushTransparency=false
      roundingType=HalfHeightRadius
      width=0
      color {colorUseRule=UseColor_Specified, specifiedColor=<rgba 0 0 0 0>}
      cornerRadii=<0 0 0 0>
     }
     tintColor {colorUseRule=UseColor_Specified, specifiedColor=<rgba 1 1 1 1>}
     uVRegion {bIsValid=false, max=<0 0>, min=<0 0>}
    }
   }
   slot {class=VerticalBoxSlot, properties {horizontalAlignment=HAlign_Center, verticalAlignment=VAlign_Bottom}}
  }
 ]
}
