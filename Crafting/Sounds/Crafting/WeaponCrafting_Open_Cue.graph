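# Sound cue for opening weapon crafting, routed to the non-critical in-game UI sound class.
# A mixer layers a UI button click with slight random pitch variation over the crafting open sound, which passes through a short fixed delay and its own default modulator.
# A concurrency override allows one instance at a time and blocks new plays while it is active.
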SoundCue /Crafting/Sounds/Crafting/WeaponCrafting_Open_Cue
root=Mixer_0
nodes [
 {id=Delay_0, children [WavePlayer_3], properties {delayMax=0.08, delayMin=0.08}}
 {id=Mixer_0, children [Modulator_0,Modulator_1], properties {inputVolume [1,1]}}
 {id=Modulator_0, children [WavePlayer_2], properties {pitchMax=1.03, pitchMin=0.97, volumeMax=1, volumeMin=1}}
 {id=Modulator_1, children [Delay_0]}
 {id=WavePlayer_2, properties {soundWaveAssetPtr=/Game/Sounds/Fort_UI_Sounds/FrontEnd/New_FrontEnd/UI_FN_Button_Hover_Click_01}}
 {id=WavePlayer_3, properties {soundWaveAssetPtr=&/WeaponCrafting_Open_01}}
]
properties {
 bOverrideConcurrency=true
 maxDistance=2097152
 soundClassObject=&Fort_UI_InGame_NonCritical
 volumeMultiplier=1.3
 concurrencyOverrides {maxCount=1, resolutionRule=PreventNew}
}
