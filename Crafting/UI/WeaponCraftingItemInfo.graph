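# Weapon crafting info panel on FortCraftingItemInfoWidget, showing an item's name, rarity, category, description, stats, ingredients and Craft button.
# OnItemRaritySet spreads the rarity palette across text, header and dynamic materials for backing, dividers and craft icon, then plays the fade/slide Intro.
# SetRarityGlowColors tints the rarity glow and scales its opacity by tier, invisible for Common.
# Two undrawn "RarityColorGetter" images also receive palette colors through their dynamic materials.
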WidgetBlueprint /Crafting/UI/WeaponCraftingItemInfo
animations [Intro {
  binding_count=3
  length_seconds=0.35
  object=&self:Intro
  sequence {bindings [
    Text_ItemDescription {
     id=#1
     possessed_class=CommonTextBlock
     tracks [{
       type=MovieSceneFloatTrack
       display_name=Render Opacity
       sections [{type=MovieSceneFloatSection, channels [None {type=MovieSceneScriptingFloatChannel, keys [[0,0],[6,1]]}]}]
      }]
    }
    Text_ItemName {
     id=#2
     possessed_class=CommonTextBlock
     tracks [
      {
       type=MovieScene2DTransformTrack
       sections [{type=MovieScene2DTransformSection, channels [Translation.X {type=MovieSceneScriptingFloatChannel, keys [[0,-24],[6,0]]}]}]
      }
      {
       type=MovieSceneFloatTrack
       display_name=Render Opacity
       sections [{type=MovieSceneFloatSection, channels [None {type=MovieSceneScriptingFloatChannel, keys [[0,0],[6,1]]}]}]
      }
     ]
    }
    Overlay_3 {
     id=#3
     possessed_class=Overlay
     tracks [
      {
       type=MovieSceneFloatTrack
       display_name=Render Opacity
       sections [{type=MovieSceneFloatSection, channels [None {type=MovieSceneScriptingFloatChannel, keys [[1,0],[7,1]]}]}]
      }
      {
       type=MovieScene2DTransformTrack
       sections [{type=MovieScene2DTransformSection, channels [Translation.X {type=MovieSceneScriptingFloatChannel, keys [[1,-24],[7,0]]}]}]
      }
     ]
    }
   ]}
 }]
graphs [
 EventGraph {nodes [
   n1 = Event OnItemRaritySet!() -> (OutputDelegate, then, Rarity:EFortRarity Enum=Common, RarityItemData:Fort Rarity Item Data Structure) {is=title K2Node_Event}
   n2 = Break Fort Rarity Item Data!(FortRarityItemData:"Fort Rarity Item Data Structure (by ref)"=@n1.RarityItemData) -> (
    Name:Text
    Color1:Linear Color Structure
    Color2:Linear Color Structure
    Color3:Linear Color Structure
    Color4:Linear Color Structure
    Color5:Linear Color Structure
    Radius:float=~
    Falloff:float=~
    Brightness:float=~
    Roughness:float=~
    Glow:float=~
   ) {is=title K2Node_BreakStruct}
   > SetColorAndOpacity(self=Get Text_ItemRarity() -> (Text_ItemRarity:Common Text Object Reference), InColorAndOpacity=Make Slate Color(SpecifiedColor=@n2.Color1))
   > SetContentColorAndOpacity!(
    execute
    self:Border Object Reference=Get CraftingIngedientsHeader() -> (CraftingIngedientsHeader:Border Object Reference)
    InContentColorAndOpacity:Linear Color Structure=@n2.Color2
   ) -> (then)
   > n3 = GetDynamicMaterial(self:Border Object Reference=Get Backing() -> (Backing:Common Border Object Reference))
   > SetVectorParameterValue(self=@n3, ParameterName=Rarity_Color, Value=@n2.Color4)
   > n4 = GetDynamicMaterial(self:Image Object Reference=Get RarityColorGetter_Rarity3() -> (RarityColorGetter_Rarity3:Image Object Reference))
   > SetVectorParameterValue(self=@n4, ParameterName=Color_Start, Value=@n2.Color3)
   > SetRarityGlowColors!(execute, self, Color:Linear Color Structure=@n2.Color3, Index:EFortRarity Enum=@n1.Rarity) -> (then)
   > n5 = GetDynamicMaterial(self=Get RarityColorGetter_Rarity5() -> (RarityColorGetter_Rarity5:Image Object Reference))
   > SetVectorParameterValue(self=@n5, ParameterName=Color_Start, Value=@n2.Color5)
   > n6 = GetDynamicMaterial(self=Get LeftLine() -> (LeftLine:Image Object Reference))
   > SetVectorParameterValue(self=@n6, ParameterName=Color_Start, Value=@n2.Color2)
   > n7 = GetDynamicMaterial(self=Get RightLine() -> (RightLine:Image Object Reference))
   > SetVectorParameterValue(self=@n7, ParameterName=Color_Start, Value=@n2.Color2)
   > PlayAnimationForward(InAnimation=Get Intro())
   > n8 = GetDynamicMaterial(self=Get CraftIcon() -> (CraftIcon:Image Object Reference))
   > SetVectorParameterValue(self=@n8, ParameterName=Hammer_Color, Value=@n2.Color2)
   > SetVectorParameterValue(self=@n8, ParameterName=Flame_Color, Value="Lerp (LinearColor)"(A=@n2.Color2, Alpha=0.25))
  ]}
 SetRarityGlowColors {nodes [
   n1 = SetRarityGlowColors!() -> (then, Color:Linear Color Structure, Index:EFortRarity Enum=Common) {is=title K2Node_FunctionEntry}
   > SetBrushTintColor!(
    execute
    self:Image Object Reference=Get Image_RarityGlow() -> (Image_RarityGlow:Image Object Reference)
    TintColor:Slate Color Structure=Make Slate Color(SpecifiedColor=@n1)
   ) -> (then)
   > SetOpacity!(
    execute
    self:Image Object Reference=Get Image_RarityGlow()
    InOpacity:float=Select!(
     Common:double=0
     Uncommon:double=0.125
     Rare:double=0.25
     Epic:double=0.333
     Legendary:double=0.666
     Mythic:double=0.666
     Transcendent:double=0.666
     Unattainable:double=0.666
     Index:EFortRarity Enum=@n1.Index
    ) -> (ReturnValue:double)
   ) -> (then)
  ]}
]
variables [RarityIndex {type=EFortRarity, default=Common}]
widgets {
 info {inheritedWidgetCount=8, parentClass=/Script/CraftingUI.FortCraftingItemInfoWidget, rootWidgetClass=&Overlay, widgetCount=35}
 widgets [
  Overlay_1 {widgetClassPath=&Overlay, widget=Overlay_1}
  Overlay_2 {
   widgetClassPath=&Overlay
   parent=Overlay_1
   widget=Overlay_2
   slot {class=OverlaySlot, properties {horizontalAlignment=HAlign_Fill, verticalAlignment=VAlign_Fill}}
  }
  Backing {
   widgetClassPath=&CommonBorder
   bIsVariable=true
   parent=Overlay_2
   widget=Backing
   properties {
    background {
     bIsDynamicallyLoaded=false
     drawAs=Image
     imageType=NoImage
     mirroring=NoMirror
     resourceName=None
     resourceObject=/Game/Athena/HUD/Inventory/Art/M_UI_HUD_CraftingPanel_Base
     tiling=NoTile
     imageSize=<32 32>
     margin=<ltrb 0 0 0 0>
     outlineSettings {
      bUseBrushTransparency=false
      roundingType=HalfHeightRadius
      width=0
      color {colorUseRule=UseColor_Specified, specifiedColor=<rgba 0 0 0 0>}
      cornerRadii=<0 0 0 0>
     }
     tintColor {colorUseRule=UseColor_Specified, specifiedColor=<rgba 1 1 1 1>}
     uVRegion {bIsValid=false, max=<0 0>, min=<0 0>}
    }
    padding=<ltrb 0 0 0 0>
   }
   slot {class=OverlaySlot, properties {horizontalAlignment=HAlign_Fill, verticalAlignment=VAlign_Fill}}
  }
  HorizontalBox_1 {
   widgetClassPath=&HorizontalBox
   parent=Backing
   widget=HorizontalBox_1
   slot {class=BorderSlot, properties {padding=<ltrb 0 0 0 0>}}
  }
  MobileLeftPadding {
   widgetClassPath=&CommonVisibilityWidgetLegacy
   parent=HorizontalBox_1
   widget=MobileLeftPadding
   properties {
    bShowForErebus=false
    bShowForGamepad=false
    bShowForMac=false
    bShowForMouseAndKeyboard=false
    bShowForPC=false
    bShowForPS4=false
    bShowForPS5=false
    bShowForSwitch2=false
    bShowForTouch=false
    bShowForXBox=false
    bShowForXSX=false
    visibility=Collapsed
   }
   slot {class=HorizontalBoxSlot, properties {horizontalAlignment=HAlign_Left}}
  }
  Spacer_114 {widgetClassPath=&Spacer, parent=MobileLeftPadding, widget=Spacer_114, properties {size=<15 0>}, slot {class=BorderSlot}}
  Overlay_0 {
   widgetClassPath=&Overlay
   parent=HorizontalBox_1
   widget=Overlay_0
   slot {class=HorizontalBoxSlot, properties {size {sizeRule=Fill, value=1}}}
  }
  VerticalBox_17 {
   widgetClassPath=&VerticalBox
   parent=Overlay_0
   widget=VerticalBox_17
   slot {class=OverlaySlot, properties {horizontalAlignment=HAlign_Fill, verticalAlignment=VAlign_Fill, padding=<ltrb 16 16 16 48>}}
  }
  HorizontalBox_31 {widgetClassPath=&HorizontalBox, parent=VerticalBox_17, widget=HorizontalBox_31, slot {class=VerticalBoxSlot}}
  VerticalBox_66 {
   widgetClassPath=&VerticalBox
   parent=HorizontalBox_31
   widget=VerticalBox_66
   slot {class=HorizontalBoxSlot, properties {verticalAlignment=VAlign_Center, size {sizeRule=Fill, value=1}}}
  }
  Text_ItemName {
   widgetClassPath=&CommonTextBlock
   bInherited=true
   parent=VerticalBox_66
   widget=Text_ItemName
   properties {
    bAutoCollapseWithEmptyText=true
    mobileFontSizeMultiplier=1.2
    scrollStyle=&TextScrollStyle-Base_C
    text="{Item Name}"
    textTransformPolicy=ToUpper
    font {
     bForceMonospaced=false
     bMaterialIsStencil=false
     fontMaterial=null
     fontObject=&Burbank
     letterSpacing=0
     monospacedWidth=1
     size=20
     skewAmount=0
     typefaceFontName=BurbankRegular-Black
     outlineSettings {
      bApplyOutlineToDropShadows=false
      bMiteredCorners=false
      bSeparateFillAlpha=false
      outlineMaterial=null
      outlineSize=0
      outlineColor=<rgba 0 0 0 1>
     }
    }
    renderTransform {angle=0, scale=<1 1>, shear=<-12 0>, translation=<0 0>}
    shadowOffset=<0 0>
   }
   slot {class=VerticalBoxSlot}
  }
  Overlay_3 {
   widgetClassPath=&Overlay
   parent=VerticalBox_66
   widget=Overlay_3
   slot {class=VerticalBoxSlot, properties {horizontalAlignment=HAlign_Left}}
  }
  Text_ItemRarity {
   widgetClassPath=&CommonTextBlock
   bInherited=true
   parent=Overlay_3
   widget=Text_ItemRarity
   properties {
    bAutoCollapseWithEmptyText=true
    mobileFontSizeMultiplier=1.3
    scrollStyle=&TextScrollStyle-Base_C
    text="{(Rarity)}"
    textTransformPolicy=ToUpper
    font {
     bForceMonospaced=false
     bMaterialIsStencil=false
     fontMaterial=null
     fontObject=&Burbank
     letterSpacing=0
     monospacedWidth=1
     size=16
     skewAmount=0
     typefaceFontName=Burbank-Black
     outlineSettings {
      bApplyOutlineToDropShadows=false
      bMiteredCorners=false
      bSeparateFillAlpha=false
      outlineMaterial=null
      outlineSize=0
      outlineColor=<rgba 0 0 0 1>
     }
    }
    shadowOffset=<0 0>
   }
   slot {class=OverlaySlot}
  }
  Image_RarityGlow {
   widgetClassPath=&Image
   bIsVariable=true
   parent=Overlay_3
   widget=Image_RarityGlow
   properties {
    visibility=SelfHitTestInvisible
    brush {
     bIsDynamicallyLoaded=false
     drawAs=Image
     imageType=NoImage
     mirroring=NoMirror
     resourceName=None
     resourceObject=/Game/Athena/HUD/LTM/Wax/Materials/M_UI_TextGlow_Add
     tiling=NoTile
     imageSize=<32 18>
     margin=<ltrb 0 0 0 0>
     outlineSettings {
      bUseBrushTransparency=false
      roundingType=HalfHeightRadius
      width=0
      color {colorUseRule=UseColor_Specified, specifiedColor=<rgba 0 0 0 0>}
      cornerRadii=<0 0 0 0>
     }
     tintColor {colorUseRule=UseColor_Specified, specifiedColor=<rgba 0 0.184 0.521 1>}
     uVRegion {bIsValid=false, max=<0 0>, min=<0 0>}
    }
    renderTransform {angle=0, scale=<1.125 1.75>, shear=<0 0>, translation=<0 0>}
   }
   slot {class=OverlaySlot, properties {horizontalAlignment=HAlign_Fill, verticalAlignment=VAlign_Center}}
  }
  SizeBox_0 {
   widgetClassPath=&SizeBox
   parent=HorizontalBox_31
   widget=SizeBox_0
   properties {bOverride_MaxDesiredWidth=true, maxDesiredWidth=200}
   slot {class=HorizontalBoxSlot, properties {verticalAlignment=VAlign_Top}}
  }
  Border_1 {
   widgetClassPath=&Border
   parent=SizeBox_0
   widget=Border_1
   properties {
    visibility=SelfHitTestInvisible
    background {
     bIsDynamicallyLoaded=false
     drawAs=Box
     imageType=NoImage
     mirroring=NoMirror
     resourceName=None
     resourceObject=/Game/Athena/HUD/Inventory/Art/M_UI_InventoryInfo_ItemType_Backing
     tiling=NoTile
     imageSize=<8 8>
     margin=<ltrb 0.5 0.5 0.5 0.5>
     outlineSettings {
      bUseBrushTransparency=false
      roundingType=HalfHeightRadius
      width=0
      color {colorUseRule=UseColor_Specified, specifiedColor=<rgba 0 0 0 0>}
      cornerRadii=<0 0 0 0>
     }
     tintColor {colorUseRule=UseColor_Specified, specifiedColor=<rgba 1 1 1 1>}
     uVRegion {bIsValid=false, max=<0 0>, min=<0 0>}
    }
    padding=<ltrb 8 4 4 4>
   }
   slot {class=SizeBoxSlot}
  }
  HorizontalBox_0 {
   widgetClassPath=&HorizontalBox
   parent=Border_1
   widget=HorizontalBox_0
   slot {class=BorderSlot, properties {padding=<ltrb 8 4 4 4>}}
  }
  Border_0 {
   widgetClassPath=&Border
   parent=HorizontalBox_0
   widget=Border_0
   properties {
    renderOpacity=0.75
    visibility=SelfHitTestInvisible
    background {
     bIsDynamicallyLoaded=false
     drawAs=NoDrawType
     imageType=NoImage
     mirroring=NoMirror
     resourceName=None
     resourceObject=null
     tiling=NoTile
     imageSize=<32 32>
     margin=<ltrb 0 0 0 0>
     outlineSettings {
      bUseBrushTransparency=false
      roundingType=HalfHeightRadius
      width=0
      color {colorUseRule=UseColor_Specified, specifiedColor=<rgba 0 0 0 0>}
      cornerRadii=<0 0 0 0>
     }
     tintColor {colorUseRule=UseColor_Specified, specifiedColor=<rgba 1 1 1 1>}
     uVRegion {bIsValid=false, max=<0 0>, min=<0 0>}
    }
    padding=<ltrb 0 0 0 0>
   }
   slot {class=HorizontalBoxSlot, properties {size {sizeRule=Fill, value=1}}}
  }
  Text_ItemCategory {
   widgetClassPath=&CommonTextBlock
   bInherited=true
   parent=Border_0
   widget=Text_ItemCategory
   properties {
    bAutoCollapseWithEmptyText=true
    mobileFontSizeMultiplier=1.3
    scrollStyle=&TextScrollStyle-Base_C
    text="{Item Category}"
    font {
     bForceMonospaced=false
     bMaterialIsStencil=false
     fontMaterial=null
     fontObject=&BurbankSmall
     letterSpacing=0
     monospacedWidth=1
     size=13
     skewAmount=0
     typefaceFontName=Bold
     outlineSettings {
      bApplyOutlineToDropShadows=false
      bMiteredCorners=false
      bSeparateFillAlpha=false
      outlineMaterial=null
      outlineSize=0
      outlineColor=<rgba 0 0 0 1>
     }
    }
    shadowOffset=<0 0>
   }
   slot {class=BorderSlot, properties {padding=<ltrb 0 0 0 0>}}
  }
  ItemCategoryIndicator {
   widgetClassPath=/Script/FortniteUI.FortItemCategoryIndicator
   bInherited=true
   parent=HorizontalBox_0
   widget=ItemCategoryIndicator
   properties {iconBrushSize=VerySmall, ordinalOfCategoryToRepresent=Tertiary, renderOpacity=0.75, visibility=SelfHitTestInvisible}
   slot {class=HorizontalBoxSlot, properties {verticalAlignment=VAlign_Center, padding=<ltrb 2 0 2 0>}}
  }
  VerticalBox_0 {
   widgetClassPath=&VerticalBox
   parent=VerticalBox_17
   widget=VerticalBox_0
   slot {class=VerticalBoxSlot, properties {size {sizeRule=Fill, value=1}}}
  }
  Text_ItemDescription {
   widgetClassPath=&CommonTextBlock
   bInherited=true
   parent=VerticalBox_0
   widget=Text_ItemDescription
   properties {
    autoWrapText=true
    bAutoCollapseWithEmptyText=true
    mobileFontSizeMultiplier=1.35
    text="{A description of the item that will be crafted telling you all about the awesome thing that you are gonna craft.}"
    font {
     bForceMonospaced=false
     bMaterialIsStencil=false
     fontMaterial=/Game/Athena/HUD/Inventory/Art/M_UI_Inventory_ItemInfo_Description_FontColor
     fontObject=&BurbankSmall
     letterSpacing=0
     monospacedWidth=1
     size=13
     skewAmount=0
     typefaceFontName=Medium
     outlineSettings {
      bApplyOutlineToDropShadows=false
      bMiteredCorners=false
      bSeparateFillAlpha=false
      outlineMaterial=null
      outlineSize=0
      outlineColor=<rgba 0 0 0 1>
     }
    }
    shadowOffset=<0 0>
   }
   slot {class=VerticalBoxSlot, properties {padding=<ltrb 0 8 0 0>, size {sizeRule=Fill, value=0.4}}}
  }
  ItemStatsWidget {
   widgetClassPath=/Game/Athena/HUD/Inventory/AthenaInventoryItemStats
   bInherited=true
   parent=VerticalBox_0
   widget=ItemStatsWidget
   slot {class=VerticalBoxSlot, properties {verticalAlignment=VAlign_Top, padding=<ltrb 0 4 0 0>}}
  }
  VerticalBox_1 {
   widgetClassPath=&VerticalBox
   parent=VerticalBox_0
   widget=VerticalBox_1
   slot {class=VerticalBoxSlot, properties {verticalAlignment=VAlign_Bottom, padding=<ltrb 0 16 0 16>, size {sizeRule=Fill, value=1}}}
  }
  HorizontalBox_2 {widgetClassPath=&HorizontalBox, parent=VerticalBox_1, widget=HorizontalBox_2, slot {class=VerticalBoxSlot}}
  LeftLine {
   widgetClassPath=&Image
   bIsVariable=true
   parent=HorizontalBox_2
   widget=LeftLine
   properties {
    renderOpacity=0.25
    visibility=SelfHitTestInvisible
    brush {
     bIsDynamicallyLoaded=false
     drawAs=Image
     imageType=NoImage
     mirroring=NoMirror
     resourceName=None
     resourceObject=/Game/Athena/HUD/Inventory/Art/M_UI_HUD_Crafting_DividerLine
     tiling=NoTile
     imageSize=<32 2>
     margin=<ltrb 0 0 0 0>
     outlineSettings {
      bUseBrushTransparency=false
      roundingType=HalfHeightRadius
      width=0
      color {colorUseRule=UseColor_Specified, specifiedColor=<rgba 0 0 0 0>}
      cornerRadii=<0 0 0 0>
     }
     tintColor {colorUseRule=UseColor_Specified, specifiedColor=<rgba 1 1 1 1>}
     uVRegion {bIsValid=false, max=<0 0>, min=<0 0>}
    }
   }
   slot {class=HorizontalBoxSlot, properties {verticalAlignment=VAlign_Center, size {sizeRule=Fill, value=0.125}}}
  }
  CraftIcon {
   widgetClassPath=&Image
   bIsVariable=true
   parent=HorizontalBox_2
   widget=CraftIcon
   properties {
    renderOpacity=0.333
    visibility=SelfHitTestInvisible
    brush {
     bIsDynamicallyLoaded=false
     drawAs=Image
     imageType=NoImage
     mirroring=NoMirror
     resourceName=None
     resourceObject=/Game/Athena/HUD/Inventory/Art/M_UI_CraftingIndicator_Panel
     tiling=NoTile
     imageSize=<20 20>
     margin=<ltrb 0 0 0 0>
     outlineSettings {
      bUseBrushTransparency=false
      roundingType=HalfHeightRadius
      width=0
      color {colorUseRule=UseColor_Specified, specifiedColor=<rgba 0 0 0 0>}
      cornerRadii=<0 0 0 0>
     }
     tintColor {colorUseRule=UseColor_Specified, specifiedColor=<rgba 1 1 1 1>}
     uVRegion {bIsValid=false, max=<0 0>, min=<0 0>}
    }
   }
   slot {class=HorizontalBoxSlot, properties {horizontalAlignment=HAlign_Center, verticalAlignment=VAlign_Center, padding=<ltrb 4 0 0 0>}}
  }
  CraftingIngedientsHeader {
   widgetClassPath=&Border
   bIsVariable=true
   parent=HorizontalBox_2
   widget=CraftingIngedientsHeader
   properties {
    horizontalAlignment=HAlign_Center
    visibility=SelfHitTestInvisible
    background {
     bIsDynamicallyLoaded=false
     drawAs=NoDrawType
     imageType=NoImage
     mirroring=NoMirror
     resourceName=None
     resourceObject=null
     tiling=NoTile
     imageSize=<32 32>
     margin=<ltrb 0 0 0 0>
     outlineSettings {
      bUseBrushTransparency=false
      roundingType=HalfHeightRadius
      width=0
      color {colorUseRule=UseColor_Specified, specifiedColor=<rgba 0 0 0 0>}
      cornerRadii=<0 0 0 0>
     }
     tintColor {colorUseRule=UseColor_Specified, specifiedColor=<rgba 1 1 1 1>}
     uVRegion {bIsValid=false, max=<0 0>, min=<0 0>}
    }
    contentColorAndOpacity=<rgba 0 0.8379 1 1>
    padding=<ltrb 0 0 0 0>
   }
   slot {class=HorizontalBoxSlot, properties {verticalAlignment=VAlign_Center, padding=<ltrb 4 0 8 0>}}
  }
  CraftingIngredientsHeader {
   widgetClassPath=&CommonTextBlock
   parent=CraftingIngedientsHeader
   widget=CraftingIngredientsHeader
   properties {
    bAutoCollapseWithEmptyText=true
    mobileFontSizeMultiplier=1.3
    scrollStyle=&TextScrollStyle-Base_C
    text=Crafting Ingredients
    textTransformPolicy=ToUpper
    font {
     bForceMonospaced=false
     bMaterialIsStencil=false
     fontMaterial=&M_UI_Inventory_ItemInfo_Description_FontColor
     fontObject=&BurbankSmall
     letterSpacing=0
     monospacedWidth=1
     size=11
     skewAmount=0
     typefaceFontName=Black
     outlineSettings {
      bApplyOutlineToDropShadows=false
      bMiteredCorners=false
      bSeparateFillAlpha=false
      outlineMaterial=null
      outlineSize=0
      outlineColor=<rgba 0 0 0 1>
     }
    }
    renderTransform {angle=0, scale=<1 1>, shear=<-8 0>, translation=<0 0>}
    shadowOffset=<0 0>
   }
   slot {class=BorderSlot, properties {horizontalAlignment=HAlign_Center, padding=<ltrb 0 0 0 0>}}
  }
  RightLine {
   widgetClassPath=&Image
   bIsVariable=true
   parent=HorizontalBox_2
   widget=RightLine
   properties {
    renderOpacity=0.25
    visibility=SelfHitTestInvisible
    brush {
     bIsDynamicallyLoaded=false
     drawAs=Image
     imageType=NoImage
     mirroring=NoMirror
     resourceName=None
     resourceObject=&M_UI_HUD_Crafting_DividerLine
     tiling=NoTile
     imageSize=<32 2>
     margin=<ltrb 0 0 0 0>
     outlineSettings {
      bUseBrushTransparency=false
      roundingType=HalfHeightRadius
      width=0
      color {colorUseRule=UseColor_Specified, specifiedColor=<rgba 0 0 0 0>}
      cornerRadii=<0 0 0 0>
     }
     tintColor {colorUseRule=UseColor_Specified, specifiedColor=<rgba 1 1 1 1>}
     uVRegion {bIsValid=false, max=<0 0>, min=<0 0>}
    }
   }
   slot {class=HorizontalBoxSlot, properties {verticalAlignment=VAlign_Center, size {sizeRule=Fill, value=1}}}
  }
  HorizontalBox_33 {
   widgetClassPath=&HorizontalBox
   parent=VerticalBox_1
   widget=HorizontalBox_33
   slot {class=VerticalBoxSlot, properties {padding=<ltrb 8 24 0 0>}}
  }
  IngredientsWidget {
   widgetClassPath=&/CraftingFormulaIngredients
   bInherited=true
   parent=HorizontalBox_33
   widget=IngredientsWidget
   slot {class=HorizontalBoxSlot, properties {horizontalAlignment=HAlign_Left}}
  }
  Button_StartCrafting {
   widgetClassPath=&IconTextButton
   bInherited=true
   parent=HorizontalBox_33
   widget=Button_StartCrafting
   properties {
    buttonText=Craft
    minHeight=48
    minWidth=144
    onButtonBaseClicked=(&self:WidgetTree.Button_StartCrafting.HandleOnButtonClicked)
    onButtonBaseDoubleClicked=(&self:WidgetTree.Button_StartCrafting.HandleOnButtonDoubleClicked)
    onButtonBaseHovered=(&self:WidgetTree.Button_StartCrafting.HandleOnButtonHovered)
    onButtonBaseUnhovered=(&self:WidgetTree.Button_StartCrafting.HandleOnButtonUnhovered)
    onSelectedChangedBase=(&self:WidgetTree.Button_StartCrafting.HandleOnSelectedChanged)
    style=/FortUILibrary/Foundation/Buttons/ButtonStyles/ButtonStyle_NewFeature_S_Yellow_nolightswipe.ButtonStyle_NewFeature_S_Yellow_nolightswipe_C
    pressedSlateSoundOverride {resourceObject=/Crafting/Sounds/Crafting/WeaponCrafting_Open_Cue}
    triggeringInputAction {dataTable=&InputActionDataTable, rowName=GenericAccept}
   }
   slot {
    class=HorizontalBoxSlot
    properties {horizontalAlignment=HAlign_Center, verticalAlignment=VAlign_Top, padding=<ltrb 20 16 30 0>, size {sizeRule=Fill, value=1}}
   }
  }
  RarityColorGetter_Rarity3 {
   widgetClassPath=&Image
   bIsVariable=true
   parent=Overlay_0
   widget=RarityColorGetter_Rarity3
   properties {brush {
     bIsDynamicallyLoaded=false
     drawAs=NoDrawType
     imageType=NoImage
     mirroring=NoMirror
     resourceName=None
     resourceObject=/Game/Athena/HUD/Inventory/Art/M_UI_CraftingPanel_RarityColor_3_getter
     tiling=NoTile
     imageSize=<32 32>
     margin=<ltrb 0 0 0 0>
     outlineSettings {
      bUseBrushTransparency=false
      roundingType=HalfHeightRadius
      width=0
      color {colorUseRule=UseColor_Specified, specifiedColor=<rgba 0 0 0 0>}
      cornerRadii=<0 0 0 0>
     }
     tintColor {colorUseRule=UseColor_Specified, specifiedColor=<rgba 1 1 1 1>}
     uVRegion {bIsValid=false, max=<0 0>, min=<0 0>}
    }}
   slot {class=OverlaySlot}
  }
  RarityColorGetter_Rarity5 {
   widgetClassPath=&Image
   bIsVariable=true
   parent=Overlay_0
   widget=RarityColorGetter_Rarity5
   properties {brush {
     bIsDynamicallyLoaded=false
     drawAs=NoDrawType
     imageType=NoImage
     mirroring=NoMirror
     resourceName=None
     resourceObject=&M_UI_CraftingPanel_RarityColor_3_getter
     tiling=NoTile
     imageSize=<32 32>
     margin=<ltrb 0 0 0 0>
     outlineSettings {
      bUseBrushTransparency=false
      roundingType=HalfHeightRadius
      width=0
      color {colorUseRule=UseColor_Specified, specifiedColor=<rgba 0 0 0 0>}
      cornerRadii=<0 0 0 0>
     }
     tintColor {colorUseRule=UseColor_Specified, specifiedColor=<rgba 1 1 1 1>}
     uVRegion {bIsValid=false, max=<0 0>, min=<0 0>}
    }}
   slot {class=OverlaySlot}
  }
 ]
}
