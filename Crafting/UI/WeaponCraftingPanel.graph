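# Weapon crafting tab widget (FortCraftingTab child) showing a recipe list, weapon item info and a quickbar editor.
# Formula-list updates swap between crafting content and a "cannot craft" overlay by formula count.
# They also copy rarity colours from WeaponCraftingItemInfo's dynamic materials, including hidden-style "getter" images, onto the recipe backing, bottom strip and darkened divider.
# Left/right list navigation cycles quickbar slots; Construct shifts the recipe column on mobile.
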WidgetBlueprint /Crafting/UI/WeaponCraftingPanel
graphs [
 EventGraph {nodes [
   n1 = Event OnFormulaListUpdated!() -> (OutputDelegate, then, NumFormulas:int=0) {is=title K2Node_Event}
   n2 = "integer > integer"(A=@n1.NumFormulas, B=0) {is=title K2Node_CallFunction}
   > SetVisibility(
    self=Get CannotCraftOverlay() -> (CannotCraftOverlay:Overlay Object Reference)
    InVisibility=Select(Option 0=SelfHitTestInvisible, Option 1=Collapsed, Index=@n2)
   )
   > SetVisibility(
    self=Get CraftingItemInfo() -> (CraftingItemInfo:Fort Crafting Item Info Widget Object Reference)
    InVisibility=Select(Option 0=Collapsed, Option 1=SelfHitTestInvisible, Index=@n2)
   )
   > SetVisibility(
    self=[Get BottomStrip() -> (BottomStrip:Image Object Reference), Get RecipesBacking() -> (RecipesBacking:Border Object Reference)]
    InVisibility=Select(Option 0=Collapsed, Option 1=SelfHitTestInvisible, Index=@n2)
   )
   > n3 = Sequence!(execute) -> (then_0, then_1, then_2:Exec)
   n4 = Cast To WeaponCraftingItemInfo!(execute=@n3.then_1, Object=Get CraftingItemInfo()) -> (then, CastFailed, AsWeapon Crafting Item Info:Weapon Crafting Item Info Object Reference) {is=title K2Node_DynamicCast}
   > n5 = GetDynamicMaterial(self=Get RarityColorGetter_Rarity3!(self:Weapon Crafting Item Info Object Reference=@n4) -> (RarityColorGetter_Rarity3:Image Object Reference))
   > n6 = Get Vector Parameter Value!(execute, self:Material Instance Dynamic Object Reference=@n5, ParameterName:Name=Color_Start) -> (then, ReturnValue:Linear Color Structure) {is=title K2Node_CallFunction}
   > n7 = GetDynamicMaterial(self=Get BottomStrip())
   > SetVectorParameterValue(self=@n7, ParameterName=Plate_Color_Start, Value=@n6)
   n8 = Cast To WeaponCraftingItemInfo(execute=@n3.then_0, Object=Get CraftingItemInfo())
   > n9 = GetDynamicMaterial(self:Border Object Reference=Get Backing!(self:Weapon Crafting Item Info Object Reference=@n8) -> (Backing:Common Border Object Reference))
   > n10 = Get Vector Parameter Value(self=@n9, ParameterName=Rarity_Color)
   > n11 = GetDynamicMaterial(self=Get RecipesBacking())
   > SetVectorParameterValue(self=@n11, ParameterName=Rarity_Color, Value=@n10)
   n12 = Cast To WeaponCraftingItemInfo(execute=@n3.then_2, Object=Get CraftingItemInfo())
   > n13 = GetDynamicMaterial(self:Image Object Reference=Get RarityColorGetter_Rarity5!(self:Weapon Crafting Item Info Object Reference=@n12) -> (RarityColorGetter_Rarity5:Image Object Reference))
   > n14 = Get Vector Parameter Value(self=@n13, ParameterName=Color_Start)
   > n15 = GetDynamicMaterial(self=Get DividerLine() -> (DividerLine:Image Object Reference))
   > SetVectorParameterValue(self=@n15, ParameterName=Color_Start, Value="Lerp (LinearColor)"(A=@n14, B=(R=0,G=0,B=0,A=1), Alpha=0.666))
   Event Construct()
   > Branch(Condition=IsMobileGame())
   > SetColumn!(
    execute
    self:Grid Slot Object Reference=SlotAsGridSlot(Widget=Get CraftRecipes() -> (CraftRecipes:Overlay Object Reference))
    InColumn:int=2
   ) -> (then)
   > SetHorizontalAlignment(
    self:Horizontal Box Slot Object Reference=SlotAsHorizontalBoxSlot!(Widget:Widget Object Reference=Get DividerLine()) -> (ReturnValue:Horizontal Box Slot Object Reference)
    InHorizontalAlignment=HAlign_Left
   )
  ]}
 ListViewNavigation {nodes [
   n1 = ListViewNavigation!() -> (then, Navigation:EUINavigation Enum) {is=title K2Node_FunctionEntry}
   > n2 = Sequence!(execute) -> (then_0, then_1)
   > n3 = Branch(Condition="Equal (Enum)"(A:EUINavigation Enum=@n1, B:EUINavigation Enum=Left))
   n4 = Get QuickbarEditor() -> (QuickbarEditor:Athena Quickbar Editor Base Object Reference)
   > SelectPreviousQuickBarSlot!(execute, self:Athena Quickbar Editor Base Object Reference=@n4, bSetFocus:bool=false) -> (then)
   Branch(execute=@n3.else, Condition="Equal (Enum)"(A=@n1, B=Right))
   > SelectNextQuickBarSlot!(execute, self:Athena Quickbar Editor Base Object Reference=@n4, bSetFocus:bool=false) -> (then)
   Return Node!(execute=@n2.then_1, ReturnValue:Widget Object Reference)
  ]}
]
variables []
widgets {
 info {inheritedWidgetCount=4, parentClass=/Script/CraftingUI.FortCraftingTab, rootWidgetClass=&Overlay, widgetCount=25}
 widgets [
  Overlay_3 {widgetClassPath=&Overlay, widget=Overlay_3}
  BackgroundBlur_100 {
   widgetClassPath=&BackgroundBlur
   parent=Overlay_3
   widget=BackgroundBlur_100
   properties {
    blurStrength=2
    horizontalAlignment=HAlign_Center
    lowQualityFallbackBrush {
     bIsDynamicallyLoaded=false
     drawAs=Image
     imageType=NoImage
     mirroring=NoMirror
     resourceName=None
     resourceObject=/Game/Athena/HUD/Inventory/Art/M_UI_InventoryPanel_Backing
     tiling=NoTile
     imageSize=<32 32>
     margin=<ltrb 0 0 0 0>
     outlineSettings {
      bUseBrushTransparency=false
      roundingType=HalfHeightRadius
      width=0
      color {colorUseRule=UseColor_Specified, specifiedColor=<rgba 0 0 0 0>}
      cornerRadii=<0 0 0 0>
     }
     tintColor {colorUseRule=UseColor_Specified, specifiedColor=<rgba 1 1 1 1>}
     uVRegion {bIsValid=false, max=<0 0>, min=<0 0>}
    }
    padding=<ltrb 0 20 0 35>
   }
   slot {class=OverlaySlot, properties {horizontalAlignment=HAlign_Fill, verticalAlignment=VAlign_Fill}}
  }
  VerticalBox_0 {
   widgetClassPath=&VerticalBox
   parent=Overlay_3
   widget=VerticalBox_0
   slot {class=OverlaySlot, properties {horizontalAlignment=HAlign_Fill, verticalAlignment=VAlign_Fill}}
  }
  Overlay {
   widgetClassPath=&Overlay
   parent=VerticalBox_0
   widget=Overlay
   slot {class=VerticalBoxSlot, properties {size {sizeRule=Fill, value=1}}}
  }
  GridPanel_0 {
   widgetClassPath=&GridPanel
   parent=Overlay
   widget=GridPanel_0
   properties {columnFill [0,0.8], rowFill [1]}
   slot {class=OverlaySlot, properties {horizontalAlignment=HAlign_Fill, verticalAlignment=VAlign_Fill}}
  }
  CraftRecipes {widgetClassPath=&Overlay, bIsVariable=true, parent=GridPanel_0, widget=CraftRecipes, slot {class=GridSlot}}
  RecipesBacking {
   widgetClassPath=&Border
   bIsVariable=true
   parent=CraftRecipes
   widget=RecipesBacking
   properties {
    horizontalAlignment=HAlign_Right
    background {
     bIsDynamicallyLoaded=false
     drawAs=Image
     imageType=NoImage
     mirroring=NoMirror
     resourceName=None
     resourceObject=/Game/Athena/HUD/Inventory/Art/M_UI_HUD_CraftingPanel_Base
     tiling=NoTile
     imageSize=<32 32>
     margin=<ltrb 0 0 0 0>
     outlineSettings {
      bUseBrushTransparency=false
      roundingType=HalfHeightRadius
      width=0
      color {colorUseRule=UseColor_Specified, specifiedColor=<rgba 0 0 0 0>}
      cornerRadii=<0 0 0 0>
     }
     tintColor {colorUseRule=UseColor_Specified, specifiedColor=<rgba 1 1 1 1>}
     uVRegion {bIsValid=false, max=<0 0>, min=<0 0>}
    }
    padding=<ltrb 0 0 0 0>
   }
   slot {class=OverlaySlot, properties {horizontalAlignment=HAlign_Fill, verticalAlignment=VAlign_Fill}}
  }
  DividerLine {
   widgetClassPath=&Image
   bIsVariable=true
   parent=RecipesBacking
   widget=DividerLine
   properties {
    visibility=SelfHitTestInvisible
    brush {
     bIsDynamicallyLoaded=false
     drawAs=Image
     imageType=NoImage
     mirroring=NoMirror
     resourceName=None
     resourceObject=/Game/Athena/HUD/Inventory/Art/M_UI_HUD_Crafting_DividerLine
     tiling=NoTile
     imageSize=<2 32>
     margin=<ltrb 0 0 0 0>
     outlineSettings {
      bUseBrushTransparency=false
      roundingType=HalfHeightRadius
      width=0
      color {colorUseRule=UseColor_Specified, specifiedColor=<rgba 0 0 0 0>}
      cornerRadii=<0 0 0 0>
     }
     tintColor {colorUseRule=UseColor_Specified, specifiedColor=<rgba 1 1 1 1>}
     uVRegion {bIsValid=false, max=<0 0>, min=<0 0>}
    }
   }
   slot {class=BorderSlot, properties {horizontalAlignment=HAlign_Right, padding=<ltrb 0 0 0 0>}}
  }
  SizeBox_0 {
   widgetClassPath=&SizeBox
   parent=CraftRecipes
   widget=SizeBox_0
   properties {bOverride_MinDesiredWidth=true, minDesiredWidth=112}
   slot {class=OverlaySlot, properties {horizontalAlignment=HAlign_Fill, verticalAlignment=VAlign_Fill}}
  }
  ListView_Recipes {
   widgetClassPath=&CommonListView
   bInherited=true
   parent=SizeBox_0
   widget=ListView_Recipes
   properties {
    entryWidgetClass=&/CraftingListEntry.CraftingListEntry_C
    horizontalEntrySpacing=8
    navigation=&self:WidgetTree.ListView_Recipes.WidgetNavigation_0
    numDesignerPreviewEntries=3
    verticalEntrySpacing=8
   }
   slot {class=SizeBoxSlot, properties {horizontalAlignment=HAlign_Center, padding=<ltrb 0 12 0 0>}}
  }
  CraftingItemInfo {
   widgetClassPath=&/WeaponCraftingItemInfo
   bInherited=true
   parent=GridPanel_0
   widget=CraftingItemInfo
   properties {bAutoActivate=true}
   slot {class=GridSlot, properties {column=1}}
  }
  BottomStrip {
   widgetClassPath=&Image
   bIsVariable=true
   parent=Overlay
   widget=BottomStrip
   properties {
    visibility=SelfHitTestInvisible
    brush {
     bIsDynamicallyLoaded=false
     drawAs=Image
     imageType=NoImage
     mirroring=NoMirror
     resourceName=None
     resourceObject=/Game/Athena/HUD/Inventory/Art/M_UI_HUD_Inventory_Crafting_Strip
     tiling=NoTile
     imageSize=<32 20>
     margin=<ltrb 0 0 0 0>
     outlineSettings {
      bUseBrushTransparency=false
      roundingType=HalfHeightRadius
      width=0
      color {colorUseRule=UseColor_Specified, specifiedColor=<rgba 0 0 0 0>}
      cornerRadii=<0 0 0 0>
     }
     tintColor {colorUseRule=UseColor_Specified, specifiedColor=<rgba 1 1 1 1>}
     uVRegion {bIsValid=false, max=<0 0>, min=<0 0>}
    }
   }
   slot {class=OverlaySlot, properties {horizontalAlignment=HAlign_Fill, verticalAlignment=VAlign_Bottom}}
  }
  CannotCraftOverlay {
   widgetClassPath=&Overlay
   bIsVariable=true
   parent=Overlay
   widget=CannotCraftOverlay
   slot {class=OverlaySlot, properties {horizontalAlignment=HAlign_Fill, verticalAlignment=VAlign_Fill}}
  }
  Image_121 {
   widgetClassPath=&Image
   parent=CannotCraftOverlay
   widget=Image_121
   properties {brush {
     bIsDynamicallyLoaded=false
     drawAs=Image
     imageType=NoImage
     mirroring=NoMirror
     resourceName=None
     resourceObject=/Game/Athena/HUD/Inventory/Art/M_UI_HUD_CraftingPanel_EmptyPanel
     tiling=NoTile
     imageSize=<32 32>
     margin=<ltrb 0 0 0 0>
     outlineSettings {
      bUseBrushTransparency=false
      roundingType=HalfHeightRadius
      width=0
      color {colorUseRule=UseColor_Specified, specifiedColor=<rgba 0 0 0 0>}
      cornerRadii=<0 0 0 0>
     }
     tintColor {colorUseRule=UseColor_Specified, specifiedColor=<rgba 1 1 1 1>}
     uVRegion {bIsValid=false, max=<0 0>, min=<0 0>}
    }}
   slot {class=OverlaySlot, properties {horizontalAlignment=HAlign_Fill, verticalAlignment=VAlign_Fill}}
  }
  VerticalBox_1 {
   widgetClassPath=&VerticalBox
   parent=CannotCraftOverlay
   widget=VerticalBox_1
   slot {class=OverlaySlot, properties {horizontalAlignment=HAlign_Center, verticalAlignment=VAlign_Center}}
  }
  HorizontalBox_114 {widgetClassPath=&HorizontalBox, parent=VerticalBox_1, widget=HorizontalBox_114, slot {class=VerticalBoxSlot}}
  Overlay_0 {
   widgetClassPath=&Overlay
   parent=HorizontalBox_114
   widget=Overlay_0
   slot {class=HorizontalBoxSlot, properties {horizontalAlignment=HAlign_Center, size {sizeRule=Fill, value=1}}}
  }
  Image_154 {
   widgetClassPath=&Image
   parent=Overlay_0
   widget=Image_154
   properties {
    renderOpacity=0.75
    visibility=SelfHitTestInvisible
    brush {
     bIsDynamicallyLoaded=false
     drawAs=Image
     imageType=NoImage
     mirroring=NoMirror
     resourceName=None
     resourceObject=/Game/Athena/HUD/Inventory/Art/M_UI_CraftingIndicator_WaterMark
     tiling=NoTile
     imageSize=<64 64>
     margin=<ltrb 0 0 0 0>
     outlineSettings {
      bUseBrushTransparency=false
      roundingType=HalfHeightRadius
      width=0
      color {colorUseRule=UseColor_Specified, specifiedColor=<rgba 0 0 0 0>}
      cornerRadii=<0 0 0 0>
     }
     tintColor {colorUseRule=UseColor_Specified, specifiedColor=<rgba 1 1 1 1>}
     uVRegion {bIsValid=false, max=<0 0>, min=<0 0>}
    }
   }
   slot {class=OverlaySlot, properties {horizontalAlignment=HAlign_Fill, verticalAlignment=VAlign_Fill}}
  }
  Image_193 {
   widgetClassPath=&Image
   parent=Overlay_0
   widget=Image_193
   properties {
    visibility=SelfHitTestInvisible
    brush {
     bIsDynamicallyLoaded=false
     drawAs=Image
     imageType=NoImage
     mirroring=NoMirror
     resourceName=None
     resourceObject=/FortUILibrary/Foundation/Art_ToolKit/AngledBoxes/M_UI_AngledFailBox_Crafting
     tiling=NoTile
     imageSize=<24 24>
     margin=<ltrb 0 0 0 0>
     outlineSettings {
      bUseBrushTransparency=false
      roundingType=HalfHeightRadius
      width=0
      color {colorUseRule=UseColor_Specified, specifiedColor=<rgba 0 0 0 0>}
      cornerRadii=<0 0 0 0>
     }
     tintColor {colorUseRule=UseColor_Specified, specifiedColor=<rgba 1 1 1 1>}
     uVRegion {bIsValid=false, max=<0 0>, min=<0 0>}
    }
    renderTransform {angle=0, scale=<1 1>, shear=<0 0>, translation=<-24 0>}
   }
   slot {class=OverlaySlot}
  }
  Text_CannotCraftAnything {
   widgetClassPath=&CommonTextBlock
   bIsVariable=true
   parent=VerticalBox_1
   widget=Text_CannotCraftAnything
   properties {
    justification=Center
    text=You cannot craft anything using the items in your loadout
    textTransformPolicy=ToUpper
    wrapTextAt=512
    colorAndOpacity {colorUseRule=UseColor_Specified, specifiedColor=<rgba 1 1 1 0.75>}
    font {
     bForceMonospaced=false
     bMaterialIsStencil=false
     fontMaterial=null
     fontObject=&BurbankSmall
     letterSpacing=0
     monospacedWidth=1
     size=16
     skewAmount=0
     typefaceFontName=Black
     outlineSettings {
      bApplyOutlineToDropShadows=false
      bMiteredCorners=false
      bSeparateFillAlpha=true
      outlineMaterial=&M-FontShadow
      outlineSize=3
      outlineColor=<rgba 0 0 0 0.5>
     }
    }
    renderTransform {angle=0, scale=<1 1>, shear=<-12 0>, translation=<0 0>}
    shadowOffset=<0 0>
   }
   slot {class=VerticalBoxSlot, properties {padding=<ltrb 0 4 0 0>}}
  }
  CannotCraftBottomStrip {
   widgetClassPath=&Image
   parent=CannotCraftOverlay
   widget=CannotCraftBottomStrip
   properties {
    visibility=SelfHitTestInvisible
    brush {
     bIsDynamicallyLoaded=false
     drawAs=Image
     imageType=NoImage
     mirroring=NoMirror
     resourceName=None
     resourceObject=/Game/Athena/HUD/Inventory/Art/M_UI_HUD_Inventory_Crafting_Srip_Disabled
     tiling=NoTile
     imageSize=<32 20>
     margin=<ltrb 0 0 0 0>
     outlineSettings {
      bUseBrushTransparency=false
      roundingType=HalfHeightRadius
      width=0
      color {colorUseRule=UseColor_Specified, specifiedColor=<rgba 0 0 0 0>}
      cornerRadii=<0 0 0 0>
     }
     tintColor {colorUseRule=UseColor_Specified, specifiedColor=<rgba 1 1 1 1>}
     uVRegion {bIsValid=false, max=<0 0>, min=<0 0>}
    }
   }
   slot {class=OverlaySlot, properties {horizontalAlignment=HAlign_Fill, verticalAlignment=VAlign_Bottom}}
  }
  QuickbarEditor {
   widgetClassPath=/Game/Athena/HUD/Inventory/AthenaQuickBarEditorWidget.AthenaQuickbarEditorWidget
   bInherited=true
   parent=VerticalBox_0
   widget=QuickbarEditor
   properties {quickBarSlotButtonClass=&/AthenaCraftingQuickBarButton.AthenaCraftingQuickBarButton_C}
   slot {class=VerticalBoxSlot, properties {horizontalAlignment=HAlign_Center, verticalAlignment=VAlign_Bottom, padding=<ltrb 0 35 0 35>}}
  }
  Spacer_37 {widgetClassPath=&Spacer, parent=VerticalBox_0, widget=Spacer_37, properties {size=<1 100>}, slot {class=VerticalBoxSlot}}
  VerticalBox_2 {
   widgetClassPath=&VerticalBox
   parent=Overlay_3
   widget=VerticalBox_2
   slot {class=OverlaySlot, properties {horizontalAlignment=HAlign_Fill, verticalAlignment=VAlign_Bottom}}
  }
  ItemInfoWidget {
   widgetClassPath=/Game/Athena/HUD/Inventory/AthenaInventoryItemInfo
   bInherited=true
   bIsVariable=true
   parent=VerticalBox_2
   widget=ItemInfoWidget
   slot {class=VerticalBoxSlot}
  }
 ]
}
